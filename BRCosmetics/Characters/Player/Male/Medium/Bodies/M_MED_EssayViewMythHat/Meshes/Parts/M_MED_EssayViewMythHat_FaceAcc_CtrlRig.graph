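# Face accessory Control Rig for the EssayViewMythHat character that applies corrective offsets to hair-lock bones as the head moves.
# Construction creates offset reader controls around head and spine, whose distances become look-up and aim weights.
# Hair Correctives branches on Backwards Hat and crouching, layering ModifyTransforms poses for aiming, movement and smoothed skydive floating.
# Crouch poses also use spherical pose readers on the head relative to spine_05.
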ControlRigBlueprint /BRCosmetics/Characters/Player/Male/Medium/Bodies/M_MED_EssayViewMythHat/Meshes/Parts/M_MED_EssayViewMythHat_FaceAcc_CtrlRig
graphs [
 New Function_ContainedGraph {nodes [
   n1 = RigVMFunctionEntryNode()
   RigVMFunction_ControlFlowBranch = &RigVMFunction_ControlFlowBranch(Condition=False, ExecuteContext=@n1.ExecuteContext)
   RigVMFunctionReturnNode(ExecuteContext=@RigVMFunction_ControlFlowBranch.Completed)
   RigUnit_GetTransform = &RigUnit_GetTransform(Item=(Type=Control,Name="head_ctrl"), bInitial=False)
   RigUnit_SetTransform = &RigUnit_SetTransform(
    Item=(Type=Bone,Name="head")
    Weight=1
    bInitial=False
    Value=@RigUnit_GetTransform.Transform
    ExecutePin=@RigVMFunction_ControlFlowBranch.True
   )
   > &RigUnit_SetTransform(
    Item=(Type=Bone,Name="clavicle_r")
    Weight=1
    bInitial=False
    Value=&RigUnit_GetTransform(Item=(Type=Control,Name="clavicle_r_ctrl"), bInitial=False).Transform
   )
   > &RigUnit_SetTransform(
    Item=(Type=Bone,Name="clavicle_l")
    Weight=1
    bInitial=False
    Value=&RigUnit_GetTransform(Item=(Type=Control,Name="clavicle_l_ctrl"), bInitial=False).Transform
   )
   > &RigUnit_SetTransform(
    Item=(Type=Bone,Name="upperarm_r")
    Weight=1
    bInitial=False
    Value=&RigUnit_GetTransform(Item=(Type=Control,Name="upperarm_r_ctrl"), bInitial=False).Transform
   )
   > &RigUnit_SetTransform(
    Item=(Type=Bone,Name="upperarm_l")
    Weight=1
    bInitial=False
    Value=&RigUnit_GetTransform(Item=(Type=Control,Name="upperarm_l_ctrl"), bInitial=False).Transform
   )
  ]}
 New Function_ContainedGraph {nodes [
   RigVMFunctionEntryNode()
   > RigVMFunction_Sequence = RigVMAggregateNode()
   RigVMVariableNode(
    Value=&RigUnit_GetTransform(Item=(Type=Control,Name="Head_Reader_ctrl"), bInitial=False).Transform.Translation
    ExecuteContext=@RigVMFunction_Sequence.A
   ) {variable=Head_Reader_translation}
   RigVMFunction_MathVectorDistance = &RigVMFunction_MathVectorDistance(
    B=&RigUnit_GetTransform(Item=(Type=Control,Name="Head_Up_ctrl"), bInitial=False).Transform.Translation
    A=RigVMVariableNode() {variable=Head_Reader_translation}.Value
   )
   Remap = &RigVMFunction_MathDoubleRemap(SourceMaximum=2, SourceMinimum=21, TargetMaximum=1, TargetMinimum=0, Value=@RigVMFunction_MathVectorDistance.Result)
   > RigVMVariableNode(Value=@Remap.Result) {variable=SPR Look Up Alpha}
   n1 = &RigVMFunction_ControlFlowBranch(ExecuteContext=@RigVMFunction_Sequence.B, Condition=RigVMVariableNode() {variable=Backwards Hat}.Value)
   n2 = &RigVMFunction_ControlFlowBranch(ExecuteContext=@n1.False, Condition=RigVMVariableNode() {variable=Is Crouching}.Value)
   n3 = RigVMVariableNode() {variable=SPR Look Up Alpha}
   &RigUnit_ModifyTransforms(
    ItemToModify="((Item=(Type=Bone,Name=\"None\"),Transform=(Rotation=(X=0,Y=0,Z=0,W=1),Translation=(X=0,Y=0,Z=0),Scale3D=(X=1,Y=1,Z=1))),(Item=(Type=Bone,Name=\"base_L_backLock_1\"),Transform=(Rotation=(X=0,Y=0,Z=0,W=1),Translation=(X=0,Y=0,Z=0),Scale3D=(X=1,Y=1,Z=1))),(Item= ...(truncated)"
    WeightMinimum=0
    ExecutePin=@n2.False
    Weight=@n3.Value
   )
   > &RigUnit_ModifyTransforms(
    ItemToModify="((Item=(Type=Bone,Name=\"None\"),Transform=(Rotation=(X=0,Y=0,Z=0,W=1),Translation=(X=0,Y=0,Z=0),Scale3D=(X=1,Y=1,Z=1))),(Item=(Type=Bone,Name=\"base_M_backLock_1\"),Transform=(Rotation=(X=0,Y=0,Z=-0.3827,W=0.9239),Translation=(X=0,Y=0,Z=0),Scale3D=(X=1,Y=1,Z= ...(truncated)"
    WeightMinimum=0
    Weight=@n3.Value
   )
   > &RigUnit_ModifyTransforms(
    ItemToModify="((Item=(Type=Bone,Name=\"base_R_frontLock_3\"),Transform=(Rotation=(X=0,Y=0,Z=-0.3827,W=0.9239),Translation=(X=0,Y=1,Z=0),Scale3D=(X=1,Y=1,Z=1))),(Item=(Type=Bone,Name=\"base_R_backLock_1\"),Transform=(Rotation=(X=0,Y=0,Z=-0.3827,W=0.9239),Translation=(X=0.5,Y ...(truncated)"
    WeightMinimum=0
    Weight=@n3.Value
   )
   n4 = &RigVMFunction_MathDoubleRemap(
    SourceMaximum=3
    SourceMinimum=18
    TargetMaximum=1
    TargetMinimum=0
    Value=&RigVMFunction_MathVectorDistance(
     B=&RigUnit_GetTransform(Item=(Type=Control,Name="Head_aim_ctrl"), bInitial=False).Transform.Translation
     A=RigVMVariableNode() {variable=Head_Reader_translation}.Value
    ).Result
   )
   > &RigUnit_ModifyTransforms(
    ItemToModify="((Item=(Type=Bone,Name=\"base_L_frontLock_1\"),Transform=(Rotation=(X=0,Y=0,Z=0,W=1),Translation=(X=0,Y=1,Z=0),Scale3D=(X=1,Y=1,Z=1))),(Item=(Type=Bone,Name=\"base_L_backLock_1\"),Transform=(Rotation=(X=0,Y=0,Z=0,W=1),Translation=(X=0,Y=0,Z=0),Scale3D=(X=1,Y=1 ...(truncated)"
    WeightMinimum=0
    Weight=@n4.Result
   )
   > &RigUnit_ModifyTransforms(
    ItemToModify="((Item=(Type=Bone,Name=\"None\"),Transform=(Rotation=(X=0,Y=0,Z=0,W=1),Translation=(X=0,Y=0,Z=0),Scale3D=(X=1,Y=1,Z=1))),(Item=(Type=Bone,Name=\"base_M_backLock_1\"),Transform=(Rotation=(X=0,Y=0,Z=0.7071,W=-0.7071),Translation=(X=0,Y=1.0315,Z=1.8946),Scale3D=( ...(truncated)"
    WeightMinimum=0
    Weight=@n4.Result
   )
   > &RigUnit_ModifyTransforms(
    ItemToModify="((Item=(Type=Bone,Name=\"base_R_frontLock_1\"),Transform=(Rotation=(X=0,Y=0,Z=0,W=1),Translation=(X=0,Y=1,Z=0),Scale3D=(X=1,Y=1,Z=1))),(Item=(Type=Bone,Name=\"base_R_backLock_1\"),Transform=(Rotation=(X=0,Y=0,Z=0,W=1),Translation=(X=0,Y=0,Z=0),Scale3D=(X=1,Y=1 ...(truncated)"
    WeightMinimum=0
    Weight=@n4.Result
   )
   n5 = RigVMVariableNode() {variable=Get is Skydiving Floating}
   AlphaInterp = &RigVMFunction_AlphaInterp(
    InterpSpeedDecreasing=10
    InterpSpeedIncreasing=10
    Scale=1
    bClampResult=false
    bMapRange=false
    Value=&RigVMFunction_MathBoolToFloat(Value=@n5.Value).Result
   )
   RigVMFunction_AnimEvalRichCurve = &RigVMFunction_AnimEvalRichCurve(
    Curve=(EditorCurveData=(Keys=((InterpMode=RCIM_Cubic),(InterpMode=RCIM_Cubic,Time=1,Value=1))))
    SourceMinimum=0
    TargetMinimum=0
    Value=@AlphaInterp.Result
   )
   n6 = &RigVMFunction_MathDoubleMul(A=@RigVMFunction_AnimEvalRichCurve.Result, B=RigVMVariableNode() {variable=SPR Look Up Alpha}.Value)
   > &RigUnit_ModifyTransforms(
    ItemToModify="((Item=(Type=Bone,Name=\"None\"),Transform=(Rotation=(X=0,Y=0,Z=0,W=1),Translation=(X=0,Y=0,Z=0),Scale3D=(X=1,Y=1,Z=1))),(Item=(Type=Bone,Name=\"base_L_backLock_1\"),Transform=(Rotation=(X=0,Y=0,Z=0.1736,W=0.9848),Translation=(X=0,Y=0,Z=0),Scale3D=(X=1,Y=1,Z=1 ...(truncated)"
    WeightMinimum=0
    Weight=@n6.Result
   )
   > &RigUnit_ModifyTransforms(
    ItemToModify="((Item=(Type=Bone,Name=\"None\"),Transform=(Rotation=(X=0,Y=0,Z=0,W=1),Translation=(X=0,Y=0,Z=0),Scale3D=(X=1,Y=1,Z=1))),(Item=(Type=Bone,Name=\"base_M_backLock_1\"),Transform=(Rotation=(X=0,Y=0,Z=-0.342,W=-0.9397),Translation=(X=0,Y=0,Z=0),Scale3D=(X=1,Y=1,Z= ...(truncated)"
    WeightMinimum=0
    Weight=@n6.Result
   )
   > n7 = &RigUnit_ModifyTransforms(
    ItemToModify="((Item=(Type=Bone,Name=\"None\"),Transform=(Rotation=(X=0,Y=0,Z=0,W=1),Translation=(X=0,Y=0,Z=0),Scale3D=(X=1,Y=1,Z=1))),(Item=(Type=Bone,Name=\"base_R_backLock_1\"),Transform=(Rotation=(X=0,Y=0,Z=0.342,W=0.9397),Translation=(X=0,Y=0,Z=0),Scale3D=(X=1,Y=1,Z=1) ...(truncated)"
    WeightMinimum=0
    Weight=@n6.Result
   )
   n8 = &RigVMFunction_ControlFlowBranch(ExecuteContext=@n7.ExecutePin, Condition=RigVMVariableNode() {variable=Is Moving 2D}.Value)
   n9 = RigVMVariableNode() {variable=SPR Look Up Alpha}
   &RigUnit_ModifyTransforms(
    ItemToModify="((Item=(Type=Bone,Name=\"None\"),Transform=(Rotation=(X=0,Y=0,Z=0,W=1),Translation=(X=0,Y=0,Z=0),Scale3D=(X=1,Y=1,Z=1))),(Item=(Type=Bone,Name=\"base_L_backLock_1\"),Transform=(Rotation=(X=0,Y=0,Z=0,W=1),Translation=(X=0,Y=0,Z=0),Scale3D=(X=1,Y=1,Z=1))),(Item= ...(truncated)"
    WeightMinimum=0
    ExecutePin=@n8.True
    Weight=@n9.Value
   )
   > &RigUnit_ModifyTransforms(
    ItemToModify="((Item=(Type=Bone,Name=\"None\"),Transform=(Rotation=(X=0,Y=0,Z=0,W=1),Translation=(X=0,Y=0,Z=0),Scale3D=(X=1,Y=1,Z=1))),(Item=(Type=Bone,Name=\"base_M_backLock_1\"),Transform=(Rotation=(X=0,Y=0.2164,Z=0,W=-0.9763),Translation=(X=0,Y=0,Z=0),Scale3D=(X=1,Y=1,Z= ...(truncated)"
    WeightMinimum=0
    Weight=@n9.Value
   )
   > &RigUnit_ModifyTransforms(
    ItemToModify="((Item=(Type=Bone,Name=\"None\"),Transform=(Rotation=(X=0,Y=0,Z=0,W=1),Translation=(X=0,Y=0,Z=0),Scale3D=(X=1,Y=1,Z=1))),(Item=(Type=Bone,Name=\"base_R_backLock_1\"),Transform=(Rotation=(X=0,Y=0,Z=0,W=1),Translation=(X=0,Y=0,Z=0),Scale3D=(X=1,Y=1,Z=1))),(Item= ...(truncated)"
    WeightMinimum=0
    Weight=@n9.Value
   )
   n10 = &RigUnit_SphericalPoseReader(
    ActiveRegionScaleFactors=(PositiveWidth=3.2,NegativeWidth=1,PositiveHeight=1,NegativeHeight=1)
    ActiveRegionSize=0.1
    Debug=(bDrawDebug=true,bDraw2D=false,bDrawLocalAxes=False,DebugScale=25,DebugSegments=20,DebugThickness=0.25)
    DriverAxis=(X=1,Y=0,Z=0)
    DriverItem=(Type=Bone,Name="head")
    FalloffRegionScaleFactors=(PositiveWidth=1.4,NegativeWidth=1,PositiveHeight=1,NegativeHeight=1)
    FalloffSize=0.4
    FlipHeightScaling=False
    FlipWidthScaling=False
    OptionalParentItem=(Type=Bone,Name="spine_05")
    RotationOffset=(X=-20,Y=0,Z=0)
    ExecutePin=@n2.True
   )
   > &RigUnit_ModifyTransforms(
    ItemToModify="((Item=(Type=Bone,Name=\"base_L_frontLock_1\"),Transform=(Rotation=(X=0,Y=0,Z=0,W=1),Translation=(X=0,Y=0,Z=0),Scale3D=(X=1,Y=1,Z=1))),(Item=(Type=Bone,Name=\"base_L_backLock_1\"),Transform=(Rotation=(X=0,Y=0,Z=0,W=1),Translation=(X=0,Y=0,Z=0),Scale3D=(X=1,Y=1 ...(truncated)"
    WeightMinimum=0
    Weight=@n10.OutputParam
   )
   > &RigUnit_ModifyTransforms(
    ItemToModify="((Item=(Type=Bone,Name=\"None\"),Transform=(Rotation=(X=0,Y=0,Z=0,W=1),Translation=(X=0,Y=0,Z=0),Scale3D=(X=1,Y=1,Z=1))),(Item=(Type=Bone,Name=\"base_M_backLock_1\"),Transform=(Rotation=(X=0.0872,Y=0,Z=0,W=0.9962),Translation=(X=0,Y=0,Z=0),Scale3D=(X=1,Y=1,Z=1 ...(truncated)"
    WeightMinimum=0
    Weight=@n10.OutputParam
   )
   > n11 = &RigUnit_ModifyTransforms(
    ItemToModify="((Item=(Type=Bone,Name=\"base_R_frontLock_1\"),Transform=(Rotation=(X=0,Y=0,Z=0,W=-1),Translation=(X=0,Y=1.6276,Z=0),Scale3D=(X=1,Y=1,Z=1))),(Item=(Type=Bone,Name=\"base_R_backLock_1\"),Transform=(Rotation=(X=0,Y=-0.4226,Z=0,W=-0.9063),Translation=(X=0,Y=0,Z=0 ...(truncated)"
    WeightMinimum=0
    Weight=@n10.OutputParam
   )
   n12 = &RigVMFunction_ControlFlowBranch(ExecuteContext=@n11.ExecutePin, Condition=RigVMVariableNode() {variable=Is Crouch Aiming Moving}.Value)
   n13 = &RigVMFunction_MathDoubleRemap(
    SourceMaximum=3
    SourceMinimum=18
    TargetMaximum=1
    TargetMinimum=0
    Value=&RigVMFunction_MathVectorDistance(
     B=&RigUnit_GetTransform(Item=(Type=Control,Name="Head_Crouch_Aim_ctrl"), bInitial=False).Transform.Translation
     A=RigVMVariableNode() {variable=Head_Reader_translation}.Value
    ).Result
   )
   &RigUnit_ModifyTransforms(
    ItemToModify="((Item=(Type=Bone,Name=\"base_L_frontLock_1\"),Transform=(Rotation=(X=0,Y=0,Z=0,W=1),Translation=(X=0,Y=0,Z=0),Scale3D=(X=1,Y=1,Z=1))),(Item=(Type=Bone,Name=\"base_L_backLock_1\"),Transform=(Rotation=(X=0,Y=0,Z=0,W=1),Translation=(X=0,Y=0,Z=0),Scale3D=(X=1,Y=1 ...(truncated)"
    WeightMinimum=0
    ExecutePin=@n12.True
    Weight=@n13.Result
   )
   > &RigUnit_ModifyTransforms(
    ItemToModify="((Item=(Type=Bone,Name=\"None\"),Transform=(Rotation=(X=0,Y=0,Z=0,W=1),Translation=(X=0,Y=0,Z=0),Scale3D=(X=1,Y=1,Z=1))),(Item=(Type=Bone,Name=\"base_M_backLock_1\"),Transform=(Rotation=(X=0,Y=0,Z=-0.866,W=0.5),Translation=(X=0,Y=0,Z=0),Scale3D=(X=1,Y=1,Z=1))) ...(truncated)"
    WeightMinimum=0
    Weight=@n13.Result
   )
   > &RigUnit_ModifyTransforms(
    ItemToModify="((Item=(Type=Bone,Name=\"base_R_frontLock_1\"),Transform=(Rotation=(X=0,Y=0,Z=0,W=1),Translation=(X=0,Y=0,Z=0),Scale3D=(X=1,Y=1,Z=1))),(Item=(Type=Bone,Name=\"base_R_backLock_1\"),Transform=(Rotation=(X=0,Y=0,Z=-0.3827,W=0.9239),Translation=(X=0,Y=0,Z=0),Scale ...(truncated)"
    WeightMinimum=0
    Weight=@n13.Result
   )
   RigVMFunction_ControlFlowBranch = &RigVMFunction_ControlFlowBranch(ExecuteContext=@n1.True, Condition=RigVMVariableNode() {variable=Is Crouching}.Value)
   n14 = &RigUnit_SphericalPoseReader(
    ActiveRegionScaleFactors=(PositiveWidth=3.2,NegativeWidth=1,PositiveHeight=1,NegativeHeight=1)
    ActiveRegionSize=0.1
    Debug=(bDrawDebug=true,bDraw2D=false,bDrawLocalAxes=False,DebugScale=25,DebugSegments=20,DebugThickness=0.25)
    DriverAxis=(X=1,Y=0,Z=0)
    DriverItem=(Type=Bone,Name="head")
    FalloffRegionScaleFactors=(PositiveWidth=1.4,NegativeWidth=1,PositiveHeight=1,NegativeHeight=1)
    FalloffSize=0.4
    FlipHeightScaling=False
    FlipWidthScaling=False
    OptionalParentItem=(Type=Bone,Name="spine_05")
    RotationOffset=(X=-20,Y=0,Z=0)
    ExecutePin=@RigVMFunction_ControlFlowBranch.True
   )
   > &RigUnit_ModifyTransforms(
    ItemToModify="((Item=(Type=Bone,Name=\"base_L_frontLock_1\"),Transform=(Rotation=(X=0,Y=0,Z=0,W=1),Translation=(X=0,Y=0,Z=0),Scale3D=(X=1,Y=1,Z=1))),(Item=(Type=Bone,Name=\"base_L_backLock_1\"),Transform=(Rotation=(X=0,Y=0,Z=0,W=1),Translation=(X=0,Y=0,Z=0),Scale3D=(X=1,Y=1 ...(truncated)"
    WeightMinimum=0
    Weight=@n14.OutputParam
   )
   > &RigUnit_ModifyTransforms(
    ItemToModify="((Item=(Type=Bone,Name=\"None\"),Transform=(Rotation=(X=0,Y=0,Z=0,W=1),Translation=(X=0,Y=0,Z=0),Scale3D=(X=1,Y=1,Z=1))),(Item=(Type=Bone,Name=\"base_M_backLock_1\"),Transform=(Rotation=(X=0,Y=0,Z=0,W=-1),Translation=(X=0,Y=0,Z=0),Scale3D=(X=1,Y=1,Z=1))),(Item ...(truncated)"
    WeightMinimum=0
    Weight=@n14.OutputParam
   )
   > n15 = &RigUnit_ModifyTransforms(
    ItemToModify="((Item=(Type=Bone,Name=\"base_R_frontLock_1\"),Transform=(Rotation=(X=0,Y=0,Z=0,W=-1),Translation=(X=0,Y=1.6276,Z=0),Scale3D=(X=1,Y=1,Z=1))),(Item=(Type=Bone,Name=\"base_R_backLock_1\"),Transform=(Rotation=(X=0,Y=0.4226,Z=0,W=0.9063),Translation=(X=0,Y=0,Z=0), ...(truncated)"
    WeightMinimum=0
    Weight=@n14.OutputParam
   )
   n16 = &RigVMFunction_ControlFlowBranch(ExecuteContext=@n15.ExecutePin, Condition=RigVMVariableNode() {variable=Is Crouch Aiming Moving}.Value)
   n17 = &RigUnit_SphericalPoseReader(
    ActiveRegionScaleFactors=(PositiveWidth=3.2,NegativeWidth=1,PositiveHeight=1,NegativeHeight=1)
    ActiveRegionSize=0.1
    Debug=(bDrawDebug=true,bDraw2D=false,bDrawLocalAxes=False,DebugScale=25,DebugSegments=20,DebugThickness=0.25)
    DriverAxis=(X=1,Y=0,Z=0)
    DriverItem=(Type=Bone,Name="head")
    FalloffRegionScaleFactors=(PositiveWidth=1.4,NegativeWidth=1,PositiveHeight=1,NegativeHeight=1)
    FalloffSize=0.4
    FlipHeightScaling=False
    FlipWidthScaling=False
    OptionalParentItem=(Type=Bone,Name="spine_05")
    RotationOffset=(X=-20,Y=0,Z=0)
    ExecutePin=@n16.True
   )
   > &RigUnit_ModifyTransforms(
    ItemToModify="((Item=(Type=Bone,Name=\"base_L_frontLock_1\"),Transform=(Rotation=(X=0,Y=0,Z=0,W=1),Translation=(X=0,Y=0,Z=0),Scale3D=(X=1,Y=1,Z=1))),(Item=(Type=Bone,Name=\"base_L_backLock_1\"),Transform=(Rotation=(X=0,Y=0,Z=0,W=-1),Translation=(X=0,Y=0,Z=0),Scale3D=(X=1,Y= ...(truncated)"
    WeightMinimum=0
    Weight=@n17.OutputParam
   )
   > &RigUnit_ModifyTransforms(
    ItemToModify="((Item=(Type=Bone,Name=\"None\"),Transform=(Rotation=(X=0,Y=0,Z=0,W=1),Translation=(X=0,Y=0,Z=0),Scale3D=(X=1,Y=1,Z=1))),(Item=(Type=Bone,Name=\"base_M_backLock_1\"),Transform=(Rotation=(X=0,Y=0,Z=0,W=-1),Translation=(X=0,Y=0,Z=0),Scale3D=(X=1,Y=1,Z=1))),(Item ...(truncated)"
    WeightMinimum=0
    Weight=@n17.OutputParam
   )
   > &RigUnit_ModifyTransforms(
    ItemToModify="((Item=(Type=Bone,Name=\"base_R_frontLock_1\"),Transform=(Rotation=(X=0,Y=0,Z=0,W=1),Translation=(X=0,Y=0,Z=0),Scale3D=(X=1,Y=1,Z=1))),(Item=(Type=Bone,Name=\"base_R_backLock_1\"),Transform=(Rotation=(X=0,Y=0,Z=0,W=1),Translation=(X=0,Y=0,Z=0),Scale3D=(X=1,Y=1 ...(truncated)"
    WeightMinimum=0
    Weight=@n17.OutputParam
   )
   n18 = RigVMVariableNode() {variable=SPR Look Up Alpha}
   &RigUnit_ModifyTransforms(
    ItemToModify="((Item=(Type=Bone,Name=\"None\"),Transform=(Rotation=(X=0,Y=0,Z=0,W=1),Translation=(X=0,Y=0,Z=0),Scale3D=(X=1,Y=1,Z=1))),(Item=(Type=Bone,Name=\"base_L_backLock_1\"),Transform=(Rotation=(X=0,Y=0,Z=0,W=1),Translation=(X=0,Y=0,Z=0),Scale3D=(X=1,Y=1,Z=1))),(Item= ...(truncated)"
    WeightMinimum=0
    Weight=@n18.Value
    ExecutePin=@RigVMFunction_ControlFlowBranch.False
   )
   > &RigUnit_ModifyTransforms(
    ItemToModify="((Item=(Type=Bone,Name=\"None\"),Transform=(Rotation=(X=0,Y=0,Z=0,W=1),Translation=(X=0,Y=0,Z=0),Scale3D=(X=1,Y=1,Z=1))),(Item=(Type=Bone,Name=\"base_M_backLock_1\"),Transform=(Rotation=(X=0,Y=0,Z=-0.7071,W=0.7071),Translation=(X=0,Y=0,Z=0),Scale3D=(X=1,Y=1,Z= ...(truncated)"
    WeightMinimum=0
    Weight=@n18.Value
   )
   > &RigUnit_ModifyTransforms(
    ItemToModify="((Item=(Type=Bone,Name=\"base_R_frontLock_3\"),Transform=(Rotation=(X=0,Y=0,Z=0,W=-1),Translation=(X=0,Y=1,Z=0),Scale3D=(X=1,Y=1,Z=1))),(Item=(Type=Bone,Name=\"base_R_backLock_1\"),Transform=(Rotation=(X=0,Y=0,Z=-0.7071,W=0.7071),Translation=(X=0.5,Y=0,Z=0),Sc ...(truncated)"
    WeightMinimum=0
    Weight=@n18.Value
   )
   n19 = &RigVMFunction_MathDoubleRemap(
    SourceMaximum=3
    SourceMinimum=18
    TargetMaximum=1
    TargetMinimum=0
    Value=&RigVMFunction_MathVectorDistance(
     B=&RigUnit_GetTransform(Item=(Type=Control,Name="Head_aim_ctrl"), bInitial=False).Transform.Translation
     A=RigVMVariableNode() {variable=Head_Reader_translation}.Value
    ).Result
   )
   > &RigUnit_ModifyTransforms(
    ItemToModify="((Item=(Type=Bone,Name=\"base_L_frontLock_1\"),Transform=(Rotation=(X=0,Y=0,Z=0,W=1),Translation=(X=0,Y=1,Z=0),Scale3D=(X=1,Y=1,Z=1))),(Item=(Type=Bone,Name=\"base_L_backLock_1\"),Transform=(Rotation=(X=0,Y=0,Z=0,W=1),Translation=(X=0,Y=0,Z=0),Scale3D=(X=1,Y=1 ...(truncated)"
    WeightMinimum=0
    Weight=@n19.Result
   )
   > &RigUnit_ModifyTransforms(
    ItemToModify="((Item=(Type=Bone,Name=\"None\"),Transform=(Rotation=(X=0,Y=0,Z=0,W=1),Translation=(X=0,Y=0,Z=0),Scale3D=(X=1,Y=1,Z=1))),(Item=(Type=Bone,Name=\"base_M_backLock_1\"),Transform=(Rotation=(X=0,Y=0,Z=0.7071,W=-0.7071),Translation=(X=0,Y=1.0315,Z=1.8946),Scale3D=( ...(truncated)"
    WeightMinimum=0
    Weight=@n19.Result
   )
   > &RigUnit_ModifyTransforms(
    ItemToModify="((Item=(Type=Bone,Name=\"base_R_frontLock_1\"),Transform=(Rotation=(X=0,Y=0,Z=0,W=1),Translation=(X=0,Y=1,Z=0),Scale3D=(X=1,Y=1,Z=1))),(Item=(Type=Bone,Name=\"base_R_backLock_1\"),Transform=(Rotation=(X=0,Y=0,Z=0,W=1),Translation=(X=0,Y=0,Z=0),Scale3D=(X=1,Y=1 ...(truncated)"
    WeightMinimum=0
    Weight=@n19.Result
   )
   Cast = &RigVMFunction_MathBoolToFloat(Value=RigVMVariableNode() {variable=Get is Skydiving Floating}.Value)
   n20 = &RigVMFunction_AlphaInterp(InterpSpeedDecreasing=10, InterpSpeedIncreasing=10, Scale=1, bClampResult=False, Value=@Cast.Result)
   n21 = &RigVMFunction_AnimEvalRichCurve(
    Curve=(EditorCurveData=(Keys=((InterpMode=RCIM_Cubic),(InterpMode=RCIM_Cubic,Time=1,Value=1))))
    SourceMinimum=0
    TargetMinimum=0
    Value=@n20.Result
   )
   Multiply = &RigVMFunction_MathDoubleMul(A=@n21.Result, B=RigVMVariableNode() {variable=SPR Look Up Alpha}.Value)
   > &RigUnit_ModifyTransforms(
    ItemToModify="((Item=(Type=Bone,Name=\"None\"),Transform=(Rotation=(X=0,Y=0,Z=0,W=1),Translation=(X=0,Y=0,Z=0),Scale3D=(X=1,Y=1,Z=1))),(Item=(Type=Bone,Name=\"base_L_backLock_1\"),Transform=(Rotation=(X=0,Y=0,Z=0.1736,W=0.9848),Translation=(X=0,Y=0,Z=0),Scale3D=(X=1,Y=1,Z=1 ...(truncated)"
    WeightMinimum=0
    Weight=@Multiply.Result
   )
   > &RigUnit_ModifyTransforms(
    ItemToModify="((Item=(Type=Bone,Name=\"None\"),Transform=(Rotation=(X=0,Y=0,Z=0,W=1),Translation=(X=0,Y=0,Z=0),Scale3D=(X=1,Y=1,Z=1))),(Item=(Type=Bone,Name=\"base_M_backLock_1\"),Transform=(Rotation=(X=0,Y=0,Z=-0.342,W=-0.9397),Translation=(X=0,Y=0,Z=0),Scale3D=(X=1,Y=1,Z= ...(truncated)"
    WeightMinimum=0
    Weight=@Multiply.Result
   )
   > n22 = &RigUnit_ModifyTransforms(
    ItemToModify="((Item=(Type=Bone,Name=\"None\"),Transform=(Rotation=(X=0,Y=0,Z=0,W=1),Translation=(X=0,Y=0,Z=0),Scale3D=(X=1,Y=1,Z=1))),(Item=(Type=Bone,Name=\"base_R_backLock_1\"),Transform=(Rotation=(X=0,Y=0,Z=0.342,W=0.9397),Translation=(X=0,Y=0,Z=0),Scale3D=(X=1,Y=1,Z=1) ...(truncated)"
    WeightMinimum=0
    Weight=@Multiply.Result
   )
   n23 = &RigVMFunction_ControlFlowBranch(ExecuteContext=@n22.ExecutePin, Condition=RigVMVariableNode() {variable=Is Moving 2D}.Value)
   n24 = RigVMVariableNode() {variable=SPR Look Up Alpha}
   &RigUnit_ModifyTransforms(
    ItemToModify="((Item=(Type=Bone,Name=\"None\"),Transform=(Rotation=(X=0,Y=0,Z=0,W=1),Translation=(X=0,Y=0,Z=0),Scale3D=(X=1,Y=1,Z=1))),(Item=(Type=Bone,Name=\"base_L_backLock_1\"),Transform=(Rotation=(X=0,Y=0,Z=0,W=1),Translation=(X=0,Y=0,Z=0),Scale3D=(X=1,Y=1,Z=1))),(Item= ...(truncated)"
    WeightMinimum=0
    ExecutePin=@n23.True
    Weight=@n24.Value
   )
   > &RigUnit_ModifyTransforms(
    ItemToModify="((Item=(Type=Bone,Name=\"None\"),Transform=(Rotation=(X=0,Y=0,Z=0,W=1),Translation=(X=0,Y=0,Z=0),Scale3D=(X=1,Y=1,Z=1))),(Item=(Type=Bone,Name=\"base_M_backLock_1\"),Transform=(Rotation=(X=0,Y=0.2164,Z=0,W=-0.9763),Translation=(X=0,Y=0,Z=0),Scale3D=(X=1,Y=1,Z= ...(truncated)"
    WeightMinimum=0
    Weight=@n24.Value
   )
   > &RigUnit_ModifyTransforms(
    ItemToModify="((Item=(Type=Bone,Name=\"None\"),Transform=(Rotation=(X=0,Y=0,Z=0,W=1),Translation=(X=0,Y=0,Z=0),Scale3D=(X=1,Y=1,Z=1))),(Item=(Type=Bone,Name=\"base_R_backLock_1\"),Transform=(Rotation=(X=0,Y=0,Z=0,W=1),Translation=(X=0,Y=0,Z=0),Scale3D=(X=1,Y=1,Z=1))),(Item= ...(truncated)"
    WeightMinimum=0
    Weight=@n24.Value
   )
   RigVMFunctionReturnNode(ExecuteContext=@RigVMFunction_Sequence.C)
   &RigVMFunction_ControlFlowBranch(ExecuteContext=(), Condition=@n5.Value)
   RigVMVariableNode() {variable=Is Crouch Walking}
  ]}
 RigVMFunctionLibrary {nodes [
   Controls = RigVMCollapseNode(Debug=False, ExecuteContext=())
   Hair Correctives = RigVMCollapseNode(ExecuteContext=())
  ]}
 RigVMFunction_Sequence_1_ContainedGraph {nodes [
   n1 = RigVMFunctionEntryNode()
   RigVMFunction_Sequence = &RigVMFunction_Sequence(ExecuteContext=@n1.ExecuteContext)
   n2 = &RigVMFunction_Sequence(ExecuteContext=@RigVMFunction_Sequence.B)
   n3 = &RigVMFunction_Sequence(ExecuteContext=@n2.B)
   RigVMFunctionReturnNode(A=@RigVMFunction_Sequence.A, B=@n2.A, C=@n3.A, D=@n3.B)
  ]}
 RigVMModel {nodes [
   n1 = &RigUnit_PrepareForExecution()
   RigVMFunction_Sequence = &RigVMFunction_Sequence(ExecuteContext=@n1.ExecutePin)
   Control Stack at Item = RigVMFunctionReferenceNode(
    Add Bottom Null=False
    Add Null=true
    Add Secondary Control=False
    Control Suffix=_ctrl
    Item=(Name="head",Type=Bone)
    Name=Head_Reader
    Null Suffix=_null
    Parent=(Name="head",Type=Bone)
    Secondary Settings="(Shape=(bVisible=True,Name=\"Default\",Color=(R=1,G=0,B=0,A=1),Transform=(Rotation=(X=0,Y=0,Z=0,W=1),Translation=(X=0,Y=0,Z=0),Scale3D=(X=1,Y=1,Z=1))),Proxy=(bIsProxy=False,ShapeVisibility=BasedOnSelection),DisplayName=\"None\",InitialSpace=LocalSpace,bUsePref ...(truncated)"
    Settings="(Shape=(bVisible=True,Name=\"Default\",Color=(R=1,G=0,B=0,A=1),Transform=(Rotation=(X=0,Y=0,Z=0,W=1),Translation=(X=0,Y=0,Z=0),Scale3D=(X=0.5,Y=0.5,Z=0.5))),Proxy=(bIsProxy=False,ShapeVisibility=BasedOnSelection),DisplayName=\"None\",InitialSpace=LocalSpace,bU ...(truncated)"
    ExecuteContext=@RigVMFunction_Sequence.B
   )
   > SetControlOffset = /Script/ControlRig.RigUnit_SetControlOffset(
    Offset=(Rotation=(X=0,Y=0,Z=0,W=1),Translation=(X=25,Y=0,Z=0),Scale3D=(X=1,Y=1,Z=1))
    Space=LocalSpace
    Control=@"Control Stack at Item".Control.Name
   ) {is=struct RigVMUnitNode, template="SetControlOffset::Execute(in Control,io Offset,in Space)"}
   > Control Stack at Item_1 = RigVMFunctionReferenceNode(
    Add Bottom Null=False
    Add Null=true
    Add Secondary Control=False
    Control Suffix=_ctrl
    Item=(Name="head",Type=Bone)
    Name=Head_Up
    Null Suffix=_null
    Parent=(Name="spine_05",Type=Bone)
    Secondary Settings="(Shape=(bVisible=True,Name=\"Default\",Color=(R=1,G=0,B=0,A=1),Transform=(Rotation=(X=0,Y=0,Z=0,W=1),Translation=(X=0,Y=0,Z=0),Scale3D=(X=1,Y=1,Z=1))),Proxy=(bIsProxy=False,ShapeVisibility=BasedOnSelection),DisplayName=\"None\",InitialSpace=LocalSpace,bUsePref ...(truncated)"
    Settings="(Shape=(bVisible=True,Name=\"Default\",Color=(R=0,G=0,B=1,A=1),Transform=(Rotation=(X=0,Y=0,Z=0,W=1),Translation=(X=0,Y=0,Z=0),Scale3D=(X=0.5,Y=0.5,Z=0.5))),Proxy=(bIsProxy=False,ShapeVisibility=BasedOnSelection),DisplayName=\"None\",InitialSpace=LocalSpace,bU ...(truncated)"
   )
   > &RigUnit_SetControlOffset(
    Offset=(Rotation=(X=0,Y=0,Z=0,W=1),Translation=(X=16,Y=-19.2,Z=0),Scale3D=(X=1,Y=1,Z=1))
    Space=LocalSpace
    Control=@"Control Stack at Item_1".Control.Name
   ) {template="SetControlOffset::Execute(in Control,io Offset,in Space)"}
   > Control Stack at Item_2 = RigVMFunctionReferenceNode(
    Add Bottom Null=False
    Add Null=true
    Add Secondary Control=False
    Control Suffix=_ctrl
    Item=(Name="head",Type=Bone)
    Name=Head_aim
    Null Suffix=_null
    Parent=(Name="spine_04",Type=Bone)
    Secondary Settings="(Shape=(bVisible=True,Name=\"Default\",Color=(R=1,G=0,B=0,A=1),Transform=(Rotation=(X=0,Y=0,Z=0,W=1),Translation=(X=0,Y=0,Z=0),Scale3D=(X=1,Y=1,Z=1))),Proxy=(bIsProxy=False,ShapeVisibility=BasedOnSelection),DisplayName=\"None\",InitialSpace=LocalSpace,bUsePref ...(truncated)"
    Settings="(Shape=(bVisible=True,Name=\"Default\",Color=(R=1,G=0,B=1,A=1),Transform=(Rotation=(X=0,Y=0,Z=0,W=1),Translation=(X=0,Y=0,Z=0),Scale3D=(X=0.5,Y=0.5,Z=0.5))),Proxy=(bIsProxy=False,ShapeVisibility=BasedOnSelection),DisplayName=\"None\",InitialSpace=LocalSpace,bU ...(truncated)"
   )
   > &RigUnit_SetControlOffset(
    Offset=(Rotation=(X=0,Y=0,Z=0,W=1),Translation=(X=16,Y=-10,Z=20),Scale3D=(X=1,Y=1,Z=1))
    Space=LocalSpace
    Control=@"Control Stack at Item_2".Control.Name
   ) {template="SetControlOffset::Execute(in Control,io Offset,in Space)"}
   > Control Stack at Item_3 = RigVMFunctionReferenceNode(
    Add Bottom Null=False
    Add Null=true
    Add Secondary Control=False
    Control Suffix=_ctrl
    Item=(Name="head",Type=Bone)
    Name=Head_Crouch_Aim
    Null Suffix=_null
    Parent=(Name="spine_04",Type=Bone)
    Secondary Settings="(Shape=(bVisible=True,Name=\"Default\",Color=(R=1,G=0,B=0,A=1),Transform=(Rotation=(X=0,Y=0,Z=0,W=1),Translation=(X=0,Y=0,Z=0),Scale3D=(X=1,Y=1,Z=1))),Proxy=(bIsProxy=False,ShapeVisibility=BasedOnSelection),DisplayName=\"None\",InitialSpace=LocalSpace,bUsePref ...(truncated)"
    Settings="(Shape=(bVisible=True,Name=\"Default\",Color=(R=0,G=0.9,B=0.9,A=1),Transform=(Rotation=(X=0,Y=0,Z=0,W=1),Translation=(X=0,Y=0,Z=0),Scale3D=(X=0.5,Y=0.5,Z=0.5))),Proxy=(bIsProxy=False,ShapeVisibility=BasedOnSelection),DisplayName=\"None\",InitialSpace=LocalSpac ...(truncated)"
   )
   > &RigUnit_SetControlOffset(Control=Head_Crouch_Aim_ctrl, Offset=(Rotation=(X=0,Y=0,Z=0,W=1),Translation=(X=-3,Y=-32,Z=20),Scale3D=(X=1,Y=1,Z=1)), Space=LocalSpace) {template="SetControlOffset::Execute(in Control,io Offset,in Space)"}
   RigUnit_BeginExecution = &RigUnit_BeginExecution()
   Hair Correctives = RigVMFunctionReferenceNode(ExecuteContext=@RigUnit_BeginExecution.ExecutePin)
   &RigUnit_HierarchyImportFromSkeleton(ExecutePin=(), NameSpace=None, bIncludeCurves=False, bIncludeMeshSockets=False, bIncludeVirtualBones=True)
   addFKCtrl = RigVMFunctionReferenceNode(
    Custom_FKCtrl=()
    ExecuteContext=()
    Options_FKCtrl="(pelvis_1_#1=False,spine_3_#2=False,clavicle_5_#3=true,arm_7_#4=true,leg_9_#5=False,head_11_46DC15EF465CD2 ...(truncated)"
    Shape_Settings="(pelvis_2_#6=(bVisible=True,Name=\"Hexagon_Thin\",Color=(R=0.0117,G=0.8281,B=1,A=1),Transform=(Rotation=(X=0,Y=-0.7071,Z=0,W=0.7071),Translation=(X=2,Y=0,Z=0),Scale3D=(X=5,Y=5,Z=7))),spine_4_#7=(bVi ...(truncated)"
   )
   setFKCtrl = RigVMFunctionReferenceNode(ExecuteContext=(), setFK_Ctrl=false)
  ]}
]
hierarchy [
 root {loc=<0 0 0>, rot=<0 0 0 1>}
 pelvis {parent=root, loc=<0 -0.7071 93.1191>, rot=<0.0127 0.707 -0.0127 -0.707>}
 spine_01 {parent=pelvis, loc=<6.7494 -0.6388 0>, rot=<0 0 0.034 -0.9994>}
 spine_02 {parent=spine_01, loc=<7.2518 -0.035 0>, rot=<0 0 0.0238 -0.9997>}
 spine_03 {parent=spine_02, loc=<6.0104 0 0>, rot=<0 0 0.0149 -0.9999>}
 spine_04 {parent=spine_03, loc=<8.1024 0 0>, rot=<0 0 -0.0132 -0.9999>}
 spine_05 {parent=spine_04, loc=<10.0146 0 0>, rot=<0 0 -0.0015 -1>}
 neck_01 {parent=spine_05, loc=<11.3138 0.1529 0>, rot=<0 0 -0.2003 -0.9797>}
 neck_02 {parent=neck_01, loc=<4.5627 0 0>, rot=<0 0 0.0504 -0.9987>}
 head {parent=neck_02, loc=<5.3914 0 0>, rot=<0 0 0.115 -0.9934>}
 base_L_backLock_2 {parent=head, loc=<5.1162 -6.6569 -4.7015>, rot=<-0.034 -0.9124 -0.4078 -0.0076>}
 dyn_L_backLock_2_01 {parent=base_L_backLock_2, loc=<0 0 0>, rot=<0 0 0 1>}
 dyn_L_backLock_2_03 {parent=dyn_L_backLock_2_01, loc=<6.2432 0 0>, rot=<0.0914 0.0524 0.1712 -0.9796>}
 dyn_L_backLock_2_05 {parent=dyn_L_backLock_2_03, loc=<6.1195 -0.8701 0>, rot=<-0.0188 -0.2272 0.1932 -0.9543>}
 dyn_L_backLock_2_06 {parent=dyn_L_backLock_2_05, loc=<3.1216 0 0>, rot=<0 0 0 1>}
 base_L_backLock_1 {parent=head, loc=<6.2775 -6.368 -4.8662>, rot=<0.0236 -0.8977 -0.431 -0.0884>}
 dyn_L_backLock_1_01 {parent=base_L_backLock_1, loc=<0 0 0>, rot=<0 0 0 1>}
 dyn_L_backLock_1_03 {parent=dyn_L_backLock_1_01, loc=<9.4215 0 0.9461>, rot=<-0.0034 0.2022 0.0165 -0.9792>}
 dyn_L_backLock_1_04 {parent=dyn_L_backLock_1_03, loc=<4.7582 0 0>, rot=<0.0456 0.2038 -0.1984 -0.9576>}
 dyn_L_backLock_1_05 {parent=dyn_L_backLock_1_04, loc=<4.7582 0 0>, rot=<0 0 -0.158 -0.9874>}
 base_L_frontLock_4 {parent=head, loc=<6.4112 -3.9213 -6.8397>, rot=<0.0748 -0.8989 -0.4276 -0.0592>}
 dyn_L_frontLock_4_01 {parent=base_L_frontLock_4, loc=<0 0 0>, rot=<0 0 0 1>}
 dyn_L_frontLock_4_03 {parent=dyn_L_frontLock_4_01, loc=<6.4724 0.8451 0.106>, rot=<0.245 0.1098 -0.0746 -0.9604>}
 dyn_L_frontLock_4_05 {parent=dyn_L_frontLock_4_03, loc=<6.4391 0.0048 0.7882>, rot=<-0.0051 0.236 0.0202 -0.9715>}
 dyn_L_frontLock_4_06 {parent=dyn_L_frontLock_4_05, loc=<3.2678 0 0>, rot=<0.0085 0.1694 -0.0801 -0.9823>}
 base_L_frontLock_3 {parent=head, loc=<9.0549 0.7365 -7.557>, rot=<-0.0934 0.6121 0.785 0.0194>}
 dyn_L_frontLock_3_01 {parent=base_L_frontLock_3, loc=<0 0 0>, rot=<0 0 0 1>}
 dyn_L_frontLock_3_03 {parent=dyn_L_frontLock_3_01, loc=<8.1115 0.434 -0.2379>, rot=<-0.012 0.0981 -0.0155 -0.995>}
 dyn_L_frontLock_3_05 {parent=dyn_L_frontLock_3_03, loc=<8.014 0.1086 1.0056>, rot=<-0.1086 0.2083 -0.0399 -0.9712>}
 dyn_L_frontLock_3_06 {parent=dyn_L_frontLock_3_05, loc=<4.0708 0 0>, rot=<0.5078 -0.0333 0.0564 -0.859>}
 base_L_frontLock_2 {parent=head, loc=<8.6876 3.23 -7.3938>, rot=<-0.133 0.609 0.7816 0.0226>}
 dyn_L_frontLock_2_01 {parent=base_L_frontLock_2, loc=<0 0 0>, rot=<0 0 0 1>}
 dyn_L_frontLock_2_03 {parent=dyn_L_frontLock_2_01, loc=<8.0565 0.3644 0.601>, rot=<0.0008 0.131 -0.0449 -0.9904>}
 dyn_L_frontLock_2_04 {parent=dyn_L_frontLock_2_03, loc=<4.0589 0 0>, rot=<0.0031 -0.1079 0.0287 -0.9937>}
 dyn_L_frontLock_2_05 {parent=dyn_L_frontLock_2_04, loc=<4.0589 0 0>, rot=<0 -0.2603 0 -0.9655>}
 base_L_frontLock_1 {parent=head, loc=<7.9869 5.2527 -5.9495>, rot=<-0.1459 0.6093 0.7793 0.0125>}
 dyn_L_frontLock_1_01 {parent=base_L_frontLock_1, loc=<0 0 0>, rot=<0 0 0 1>}
 dyn_L_frontLock_1_03 {parent=dyn_L_frontLock_1_01, loc=<8.2134 0.5976 0.9343>, rot=<0.0056 0.1568 -0.0808 -0.9843>}
 dyn_L_frontLock_1_05 {parent=dyn_L_frontLock_1_03, loc=<8.3632 0 0>, rot=<0 0.154 0 -0.9881>}
 dyn_L_frontLock_1_06 {parent=dyn_L_frontLock_1_05, loc=<4.1816 0 0>, rot=<-0.0087 0.1191 -0.0078 -0.9928>}
 base_M_backLock_3 {parent=head, loc=<2.0242 -6.5893 -3.3344>, rot=<-0.0348 -0.9604 -0.2766 -0.0028>}
 dyn_M_backLock_3_01 {parent=base_M_backLock_3, loc=<0 0 0>, rot=<0 0 0 1>}
 dyn_M_backLock_3_03 {parent=dyn_M_backLock_3_01, loc=<6.104 -0.7174 -0.5271>, rot=<-0.0084 -0.1265 0.221 -0.967>}
 dyn_M_backLock_3_05 {parent=dyn_M_backLock_3_03, loc=<6.191 -0.1103 -0.5029>, rot=<-0.2215 0.0008 0.1793 -0.9585>}
 dyn_M_backLock_3_06 {parent=dyn_M_backLock_3_05, loc=<3.1169 0 0>, rot=<-0.1245 0.1746 -0.2021 -0.9556>}
 base_M_backLock_2 {parent=head, loc=<1.5257 -7.348 1.8261>, rot=<-0.0235 -0.9985 0.0223 -0.0443>}
 dyn_M_backLock_2_01 {parent=base_M_backLock_2, loc=<0 0 0>, rot=<0 0 0 1>}
 dyn_M_backLock_2_03 {parent=dyn_M_backLock_2_01, loc=<6.4464 -1.1964 -0.1376>, rot=<-0.0156 -0.1346 0.2913 -0.947>}
 dyn_M_backLock_2_05 {parent=dyn_M_backLock_2_03, loc=<6.5978 -0.2382 -0.0237>, rot=<0.0699 0.0564 -0.0023 -0.996>}
 dyn_M_backLock_2_06 {parent=dyn_M_backLock_2_05, loc=<3.3032 0 0>, rot=<-0.0525 0.0841 -0.0544 -0.9936>}
 base_M_backLock_1 {parent=head, loc=<5.6869 -8.1775 0.424>, rot=<0.0503 -0.9981 0.0288 0.0199>}
 dyn_M_backLock_1_01 {parent=base_M_backLock_1, loc=<0 0 0>, rot=<0 0 0 1>}
 dyn_M_backLock_1_03 {parent=dyn_M_backLock_1_01, loc=<7.5425 -0.2513 -0.7088>, rot=<0.001 -0.1763 0.1045 -0.9788>}
 dyn_M_backLock_1_05 {parent=dyn_M_backLock_1_03, loc=<7.5288 -0.6572 0.4852>, rot=<-0.0073 0.3911 0.1548 -0.9072>}
 base_R_backLock_2 {parent=head, loc=<6.0899 -7.4065 4.4531>, rot=<0.0394 -0.9718 0.2277 0.0473>}
 dyn_R_backLock_2_01 {parent=base_R_backLock_2, loc=<0 0 0>, rot=<0 0 0 1>}
 dyn_R_backLock_2_03 {parent=dyn_R_backLock_2_01, loc=<7.2358 0.1696 0>, rot=<0 0 0.0417 -0.9991>}
 dyn_R_backLock_2_05 {parent=dyn_R_backLock_2_03, loc=<7.0294 -1.1938 -0.2319>, rot=<0.0988 0.0494 0.1724 -0.9788>}
 dyn_R_backLock_2_06 {parent=dyn_R_backLock_2_05, loc=<3.6199 0 0>, rot=<0 0 0 1>}
 base_R_backLock_1 {parent=head, loc=<5.5553 -4.2572 6.6394>, rot=<-0.013 -0.9725 0.2299 0.0349>}
 dyn_R_backLock_1_01 {parent=base_R_backLock_1, loc=<0 0 0>, rot=<0 0 0 1>}
 dyn_R_backLock_1_03 {parent=dyn_R_backLock_1_01, loc=<7.5624 -0.5034 0.3465>, rot=<-0.0122 0.0146 0.1574 -0.9873>}
 dyn_R_backLock_1_05 {parent=dyn_R_backLock_1_03, loc=<7.3506 -1.3691 0.2141>, rot=<0.0035 0.0279 0.2429 -0.9696>}
 dyn_R_backLock_1_06 {parent=dyn_R_backLock_1_05, loc=<3.5582 0 0>, rot=<0 0 -0.0395 -0.9992>}
 base_R_frontLock_3 {parent=head, loc=<8.7627 -3.1999 7.1579>, rot=<0.1228 -0.8551 0.5029 0.0279>}
 dyn_R_frontLock_3_01 {parent=base_R_frontLock_3, loc=<0 0 0>, rot=<0 0 0 1>}
 dyn_R_frontLock_3_03 {parent=dyn_R_frontLock_3_01, loc=<7.8019 0.2461 0>, rot=<0.0034 -0.1067 -0.0313 -0.9938>}
 dyn_R_frontLock_3_05 {parent=dyn_R_frontLock_3_03, loc=<7.6283 0.8295 -0.834>, rot=<-0.1987 -0.3535 -0.2169 -0.888>}
 dyn_R_frontLock_3_06 {parent=dyn_R_frontLock_3_05, loc=<3.9048 0 0>, rot=<0 0 0 1>}
 base_R_frontLock_2 {parent=head, loc=<8.2855 1.8967 7.2006>, rot=<0.1164 -0.6093 0.7837 0.0313>}
 dyn_R_frontLock_2_01 {parent=base_R_frontLock_2, loc=<0 0 0>, rot=<0 0 0 1>}
 dyn_R_frontLock_2_03 {parent=dyn_R_frontLock_2_01, loc=<7.6789 0.5263 -0.2083>, rot=<0.0008 -0.066 -0.0683 -0.9955>}
 dyn_R_frontLock_2_04 {parent=dyn_R_frontLock_2_03, loc=<3.8603 0 0>, rot=<0 -0.1506 0 -0.9886>}
 dyn_R_frontLock_2_06 {parent=dyn_R_frontLock_2_04, loc=<7.7001 -0.398 0>, rot=<0.0224 -0.2052 0.2061 -0.9565>}
 base_R_frontLock_1 {parent=head, loc=<8.0629 5.3518 6.3138>, rot=<0.0834 -0.6074 0.7879 0.0569>}
 dyn_R_frontLock_1_01 {parent=base_R_frontLock_1, loc=<0 0 0>, rot=<0 0 0 1>}
 dyn_R_frontLock_1_03 {parent=dyn_R_frontLock_1_01, loc=<7.5095 0.5633 0.2913>, rot=<-0.0093 -0.0033 -0.1276 -0.9918>}
 dyn_R_frontLock_1_04 {parent=dyn_R_frontLock_1_03, loc=<3.7815 0 0>, rot=<0 -0.0958 0 -0.9954>}
 dyn_R_frontLock_1_06 {parent=dyn_R_frontLock_1_04, loc=<7.9243 0 -0.9154>, rot=<0.2394 -0.3165 -0.0083 -0.9178>}
 FACIAL_C_FacialRoot {parent=head, loc=<2.1524 -0.746 -0.0002>, rot=<0.4879 -0.5118 -0.4879 -0.5118>}
 FACIAL_L_CheekInner {parent=FACIAL_C_FacialRoot, loc=<2.0706 0.5064 9.7474>, rot=<-0.1233 -0.0876 -0.197 -0.9687>}
 FACIAL_R_CheekInner {parent=FACIAL_C_FacialRoot, loc=<-2.0969 0.5498 9.7343>, rot=<-0.1235 0.0883 0.1973 -0.9685>}
 FACIAL_L_CheekOuter {parent=FACIAL_C_FacialRoot, loc=<3.6422 0.8154 8.2405>, rot=<0.0549 -0.3766 0.1005 -0.9193>}
 FACIAL_R_CheekOuter {parent=FACIAL_C_FacialRoot, loc=<-3.6548 0.8608 8.2416>, rot=<0.0551 0.3773 -0.1001 -0.919>}
 FACIAL_L_NasolabialBulge {parent=FACIAL_C_FacialRoot, loc=<3.3303 4.2879 8.6317>, rot=<0.0577 -0.377 0.024 -0.9241>}
 FACIAL_R_NasolabialBulge {parent=FACIAL_C_FacialRoot, loc=<-3.333 4.3119 8.6318>, rot=<0.0578 0.3777 -0.0235 -0.9238>}
 FACIAL_L_NasolabialFurrow {parent=FACIAL_C_FacialRoot, loc=<2.9289 2.7365 10.7101>, rot=<-0.0333 -0.1254 -0.3635 -0.9225>}
 FACIAL_R_NasolabialFurrow {parent=FACIAL_C_FacialRoot, loc=<-2.9346 2.7599 10.7054>, rot=<-0.0336 0.1261 0.3638 -0.9223>}
 FACIAL_C_UnderChin {parent=FACIAL_C_FacialRoot, loc=<0.001 9.1329 5.6353>, rot=<0.4997 0.0002 0.0004 -0.8662>}
 FACIAL_L_UnderChin {parent=FACIAL_C_FacialRoot, loc=<2.929 7.4952 4.5654>, rot=<0.4023 -0.3547 0.0826 -0.84>}
 FACIAL_R_UnderChin {parent=FACIAL_C_FacialRoot, loc=<-2.9347 7.502 4.5613>, rot=<0.4025 0.3551 -0.0819 -0.8398>}
 FACIAL_C_Jaw {parent=FACIAL_C_FacialRoot, loc=<-0.0148 1.8743 3.0458>, rot=<0.3419 0.0003 0.0003 -0.9397>}
 FACIAL_C_Jawline {parent=FACIAL_C_Jaw, loc=<0.0123 2.6886 8.5724>, rot=<0.2586 0 0 -0.966>}
 FACIAL_L_ChinSide {parent=FACIAL_C_Jaw, loc=<2.7087 0.0752 8.7308>, rot=<-0.2208 -0.3738 0.0499 -0.8995>}
 FACIAL_R_ChinSide {parent=FACIAL_C_Jaw, loc=<-2.6753 0.0846 8.7301>, rot=<-0.2208 0.3738 -0.0499 -0.8995>}
 FACIAL_C_Chin {parent=FACIAL_C_Jaw, loc=<0.0123 0.8853 9.2459>, rot=<-0.2591 0 0 -0.9659>}
 FACIAL_L_Jawline {parent=FACIAL_C_FacialRoot, loc=<4.2986 5.6221 6.2583>, rot=<0.2092 0.158 -0.0544 -0.9635>}
 FACIAL_R_Jawline {parent=FACIAL_C_FacialRoot, loc=<-4.2991 5.6234 6.2565>, rot=<0.2091 -0.1575 0.0549 -0.9636>}
 FACIAL_C_LowerLipRotation {parent=FACIAL_C_FacialRoot, loc=<-0.0148 1.8743 3.0458>, rot=<0.3419 0.0003 0.0003 -0.9397>}
 FACIAL_C_MouthLower {parent=FACIAL_C_LowerLipRotation, loc=<0.0189 -1.9992 6.7083>, rot=<-0.342 0 0 -0.9397>}
 FACIAL_C_LipLower {parent=FACIAL_C_MouthLower, loc=<0.0067 1.3375 2.3828>, rot=<0.0872 0 0 -0.9962>}
 FACIAL_L_LipLower {parent=FACIAL_C_MouthLower, loc=<0.9916 1.2628 2.1>, rot=<0.0927 -0.1294 0.0272 -0.9869>}
 FACIAL_R_LipLower {parent=FACIAL_C_MouthLower, loc=<-0.9761 1.2667 2.1039>, rot=<0.0927 0.1294 -0.0272 -0.9869>}
 FACIAL_L_LipLowerOuter {parent=FACIAL_C_MouthLower, loc=<1.9737 0.7397 1.608>, rot=<0.0913 -0.2321 0.0318 -0.9679>}
 FACIAL_R_LipLowerOuter {parent=FACIAL_C_MouthLower, loc=<-1.9491 0.8011 1.5974>, rot=<0.0913 0.2321 -0.0318 -0.9679>}
 base_hatback {parent=head, loc=<6.143 -5.9801 -0.0285>, rot=<0.0029 -0.7071 -0.0029 -0.7071>}
 dyn_hatback {parent=base_hatback, loc=<0 0 0>, rot=<0 0 0 1>}
 attach_cape {parent=spine_05, loc=<9.017 -11.714 0>, rot=<0.0616 -0.7044 -0.0616 -0.7044>}
 attach_backpack {parent=spine_05, loc=<-0.123 -13.473 0>, rot=<-0.0093 -0.707 0.0093 -0.707>}
 clavicle_l {parent=spine_05, loc=<4.3575 -0.1466 -2.3223>, rot=<0.0465 -0.7289 -0.009 -0.6829>}
 upperarm_l {parent=clavicle_l, loc=<15.5111 0 0>, rot=<0.0226 -0.3234 -0.0088 -0.9459>}
 clavicle_r {parent=spine_05, loc=<4.3575 -0.1466 2.3223>, rot=<0.7289 0.0465 -0.6829 0.009>}
 upperarm_r {parent=clavicle_r, loc=<-15.5111 0 0>, rot=<0.0226 -0.3234 -0.0088 -0.9459>}
]
variables [
 Backwards Hat {type=bool, default=False}
 Get is Skydiving Floating {type=bool, default=False}
 Head_Reader_translation {type=FVector, default=(X=0,Y=0,Z=0)}
 Is Crouch Aiming Moving {type=bool, default=False}
 Is Crouch Walking {type=bool, default=False}
 Is Crouching {type=bool, default=False}
 Is Moving 2D {type=bool, default=False}
 Look Up and Left Two Hands Multiplied Alpha {type=double, default=0}
 SPR Look Up Alpha {type=double, default=0}
 Upperarm R SPR Aim Up Two Hands Alpha {type=double, default=0}
]
