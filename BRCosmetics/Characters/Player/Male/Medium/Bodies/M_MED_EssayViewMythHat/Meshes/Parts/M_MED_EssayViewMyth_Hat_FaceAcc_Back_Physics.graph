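# Physics asset for the hat's FaceAcc_Back part on a male medium player cosmetic, simulating 14 hanging locks, likely hair.
# Each lock is a capsule chain from an angularly free base joint on the head through dyn bones with tight swing and twist limits.
# Capsules on head, neck, upper spine, clavicles and upper arms provide body collision shapes.
# Three extra angularly free constraints cross-link neighboring front-lock chains, two right and one left.
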PhysicsAsset /BRCosmetics/Characters/Player/Male/Medium/Bodies/M_MED_EssayViewMythHat/Meshes/Parts/M_MED_EssayViewMyth_Hat_FaceAcc_Back_Physics
bodies {
 base_L_backLock_1 [{shapeType=Capsule, length=0.505, radius=0.505, shapeName=base_L_backLock_1_capsule, rotation=<rot 0 -90 0>}]
 base_L_backLock_2 [{
   shapeType=Capsule
   length=0.62
   radius=0.6144
   shapeName=base_L_backLock_2_capsule
   center=<1.5579 0.6801 -2.4099>
   rotation=<rot 40.0693 139.6453 -142.8534>
  }]
 base_L_frontLock_1 [{shapeType=Capsule, length=0.505, radius=0.505, shapeName=base_L_frontLock_1_capsule, rotation=<rot 0 -90 0>}]
 base_L_frontLock_2 [{shapeType=Capsule, length=0.505, radius=0.505, shapeName=base_L_frontLock_2_capsule, rotation=<rot 0 -90 0>}]
 base_L_frontLock_3 [{shapeType=Capsule, length=0.505, radius=0.505, shapeName=base_L_frontLock_3_capsule, rotation=<rot 0 -90 0>}]
 base_L_frontLock_4 [{shapeType=Capsule, length=0.505, radius=0.505, shapeName=base_L_frontLock_4_capsule, rotation=<rot 0 -90 0>}]
 base_M_backLock_1 [{shapeType=Capsule, length=0.505, radius=0.505, shapeName=base_M_backLock_1_capsule, rotation=<rot 0 -90 0>}]
 base_M_backLock_2 [{shapeType=Capsule, length=0.505, radius=0.505, shapeName=base_M_backLock_2_capsule, rotation=<rot 0 -90 0>}]
 base_M_backLock_3 [{shapeType=Capsule, length=0.505, radius=0.505, shapeName=base_M_backLock_3_capsule, rotation=<rot 0 -90 0>}]
 base_R_backLock_1 [{shapeType=Capsule, length=0.505, radius=0.505, shapeName=base_R_backLock_1_capsule, rotation=<rot 0 -90 0>}]
 base_R_backLock_2 [{
   shapeType=Capsule
   length=0.505
   radius=0.505
   shapeName=base_R_backLock_2_capsule
   center=<0.0743 1.2667 1.9659>
   rotation=<rot 0 -102.7334 -82.5195>
  }]
 base_R_frontLock_1 [{shapeType=Capsule, length=0.505, radius=0.505, shapeName=base_R_frontLock_1_capsule, rotation=<rot 0 -90 0>}]
 base_R_frontLock_2 [{shapeType=Capsule, length=0.505, radius=0.505, shapeName=base_R_frontLock_2_capsule, rotation=<rot 0 -90 0>}]
 base_R_frontLock_3 [{shapeType=Capsule, length=0.505, radius=0.505, shapeName=base_R_frontLock_3_capsule, rotation=<rot 0 -90 0>}]
 clavicle_l [{
   shapeType=Capsule
   length=7.5
   radius=6.3
   shapeName=clavicle_l_capsule
   center=<9.3481 -2.3242 0.6078>
   rotation=<rot -75 179.9997 -179.9997>
  }]
 clavicle_r [{shapeType=Capsule, length=7, radius=7.5, shapeName=clavicle_l_capsule, center=<-9.2861 1.7699 0.4228>, rotation=<rot -75 179.9997 -179.9997>}]
 dyn_L_backLock_1_01 [{shapeType=Capsule, length=7.3693, radius=0.8205, shapeName=dyn_L_backLock_1_01_capsule, center=<4.5841 0 0>, rotation=<rot 90 0 0>}]
 dyn_L_backLock_1_03 [{shapeType=Capsule, length=2.3978, radius=1.5, shapeName=dyn_L_backLock_1_03_capsule, center=<2.2833 0 0>, rotation=<rot 90 0 0>}]
 dyn_L_backLock_1_04 [{shapeType=Capsule, length=4.5351, radius=2, shapeName=dyn_L_backLock_1_04_capsule, center=<3.1045 0 0>, rotation=<rot 90 0 0>}]
 dyn_L_backLock_2_01 [{shapeType=Capsule, length=2.2714, radius=1.5845, shapeName=dyn_L_backLock_2_01_capsule, center=<2.8228 0 0>, rotation=<rot 90 0 0>}]
 dyn_L_backLock_2_03 [{shapeType=Capsule, length=4.8509, radius=1.6524, shapeName=dyn_L_backLock_2_03_capsule, center=<2.3864 -1.0075 0>, rotation=<rot 90 0 0>}]
 dyn_L_backLock_2_05 [{shapeType=Capsule, length=1.9865, radius=2.1023, shapeName=dyn_L_backLock_2_05_capsule, center=<2.2811 0 0>, rotation=<rot 90 0 0>}]
 dyn_L_backLock_2_06 [{shapeType=Capsule, length=3.0815, radius=2.0578, shapeName=dyn_L_backLock_2_06_capsule, center=<3.3953 0 0>, rotation=<rot 90 0 0>}]
 dyn_L_frontLock_1_01 [{shapeType=Capsule, length=6.0414, radius=1, shapeName=dyn_L_frontLock_1_01_capsule, center=<4.1595 -0.4884 0>, rotation=<rot 90 0 0>}]
 dyn_L_frontLock_1_03 [{shapeType=Capsule, length=3.3184, radius=1.5, shapeName=dyn_L_frontLock_1_03_capsule, center=<3.7181 0 0>, rotation=<rot 90 0 0>}]
 dyn_L_frontLock_1_05 [{shapeType=Capsule, length=2.2406, radius=1.5, shapeName=dyn_L_frontLock_1_05_capsule, center=<2.8031 0 0>, rotation=<rot 90 0 0>}]
 dyn_L_frontLock_1_06 [{shapeType=Capsule, length=2.8012, radius=1.5, shapeName=dyn_L_frontLock_1_06_capsule, center=<3.2886 1 0>, rotation=<rot 90 0 0>}]
 dyn_L_frontLock_2_01 [{shapeType=Capsule, length=6.0035, radius=1, shapeName=dyn_L_frontLock_2_01_capsule, center=<3.7617 0 0>, rotation=<rot 90 1.9692 0>}]
 dyn_L_frontLock_2_03 [{shapeType=Capsule, length=2, radius=1.001, shapeName=dyn_L_frontLock_2_03_capsule, center=<1.9738 0 0>, rotation=<rot 90 0 0>}]
 dyn_L_frontLock_2_04 [{shapeType=Capsule, length=2, radius=1, shapeName=dyn_L_frontLock_2_04_capsule, center=<1.9664 0 0>, rotation=<rot 90 0 0>}]
 dyn_L_frontLock_2_05 [{shapeType=Capsule, length=2.0803, radius=0.7509, shapeName=dyn_L_frontLock_2_05_capsule, center=<1.8707 0 0>, rotation=<rot 90 0 0>}]
 dyn_L_frontLock_3_01 [{shapeType=Capsule, length=4.5, radius=1.1675, shapeName=dyn_L_frontLock_3_01_capsule, center=<3.5221 0 0>, rotation=<rot 90 0 0>}]
 dyn_L_frontLock_3_03 [{shapeType=Capsule, length=4, radius=1.6135, shapeName=dyn_L_frontLock_3_03_capsule, center=<3.9116 0 0>, rotation=<rot 90 0 0>}]
 dyn_L_frontLock_3_05 [{shapeType=Capsule, length=2.2592, radius=1.4632, shapeName=dyn_L_frontLock_3_05_capsule, center=<2.2177 0 0>, rotation=<rot 90 0 0>}]
 dyn_L_frontLock_3_06 [{shapeType=Capsule, length=3.6146, radius=1.5, shapeName=dyn_L_frontLock_3_06_capsule, center=<3.6429 0 0>, rotation=<rot 90 0 0>}]
 dyn_L_frontLock_4_01 [{shapeType=Capsule, length=3.745, radius=1.2502, shapeName=dyn_L_frontLock_4_01_capsule, center=<3.0978 0 0>, rotation=<rot 90 0 0>}]
 dyn_L_frontLock_4_03 [{shapeType=Capsule, length=4.0335, radius=1.285, shapeName=dyn_L_frontLock_4_03_capsule, center=<3.2298 0 0>, rotation=<rot 90 0 0>}]
 dyn_L_frontLock_4_05 [{shapeType=Capsule, length=2.1285, radius=1.5031, shapeName=dyn_L_frontLock_4_05_capsule, center=<2.5364 0 0>, rotation=<rot 90 0 0>}]
 dyn_L_frontLock_4_06 [{shapeType=Capsule, length=5.5, radius=1.5, shapeName=dyn_L_frontLock_4_06_capsule, center=<5.9228 0 0>, rotation=<rot 90 0 0>}]
 dyn_M_backLock_1_01 [{shapeType=Capsule, length=5.639, radius=0.7081, shapeName=dyn_M_backLock_1_01_capsule, center=<3.7806 0 0>, rotation=<rot 90 0 0>}]
 dyn_M_backLock_1_03 [{shapeType=Capsule, length=5.457, radius=0.8536, shapeName=dyn_M_backLock_1_03_capsule, center=<3.5657 0 0>, rotation=<rot 90 0 0>}]
 dyn_M_backLock_1_05 [{shapeType=Capsule, length=4.6409, radius=0.7819, shapeName=dyn_M_backLock_1_05_capsule, center=<2.0491 0 0>, rotation=<rot 90 0 0>}]
 dyn_M_backLock_2_01 [{shapeType=Capsule, length=3.8133, radius=1.7901, shapeName=dyn_M_backLock_2_01_capsule, center=<2.5867 -1.1296 0>, rotation=<rot 90 0 0>}]
 dyn_M_backLock_2_03 [{shapeType=Capsule, length=4.1464, radius=2.2, shapeName=dyn_M_backLock_2_03_capsule, center=<3.5384 -0.4855 0>, rotation=<rot 90 0 0>}]
 dyn_M_backLock_2_05 [
  {shapeType=Capsule, length=3.9096, radius=2.0396, shapeName=dyn_M_backLock_2_05_capsule, center=<3.6237 -0.9632 0>, rotation=<rot 90 0 0>}
  {
   shapeType=Capsule
   length=5.5766
   radius=1.2629
   shapeName=dyn_M_backLock_2_05_capsule
   center=<4.3801 -0.2647 2.1554>
   rotation=<rot 90 15.2807 0>
  }
 ]
 dyn_M_backLock_3_01 [{shapeType=Capsule, length=3.5408, radius=1.5, shapeName=dyn_M_backLock_3_01_capsule, center=<3.1193 0 0>, rotation=<rot 90 0 0>}]
 dyn_M_backLock_3_03 [{shapeType=Capsule, length=2.0644, radius=1.7909, shapeName=dyn_M_backLock_3_03_capsule, center=<2.7373 0 0>, rotation=<rot 90 0 0>}]
 dyn_M_backLock_3_05 [{shapeType=Capsule, length=6.2479, radius=1.7813, shapeName=dyn_M_backLock_3_05_capsule, center=<4.2748 -0.4292 0>, rotation=<rot 90 0 0>}]
 dyn_R_backLock_1_01 [{shapeType=Capsule, length=4.8417, radius=1.2117, shapeName=dyn_R_backLock_1_01_capsule, center=<4.0586 0 0>, rotation=<rot 90 0 0>}]
 dyn_R_backLock_1_03 [{shapeType=Capsule, length=3.483, radius=1.5, shapeName=dyn_R_backLock_1_03_capsule, center=<3.4597 0 0>, rotation=<rot 90 0 0>}]
 dyn_R_backLock_1_05 [{shapeType=Capsule, length=4.7082, radius=1.5, shapeName=dyn_R_backLock_1_05_capsule, center=<2.6549 0 0>, rotation=<rot 90 0 0>}]
 dyn_R_backLock_2_01 [{shapeType=Capsule, length=4.049, radius=1.5835, shapeName=dyn_R_backLock_2_01_capsule, center=<3.467 0 0>, rotation=<rot 90 0 0>}]
 dyn_R_backLock_2_03 [{shapeType=Capsule, length=4.2392, radius=1.3943, shapeName=dyn_R_backLock_2_03_capsule, center=<3.5425 0 0>, rotation=<rot 90 0 0>}]
 dyn_R_backLock_2_05 [{shapeType=Capsule, length=3.9256, radius=1.7847, shapeName=dyn_R_backLock_2_05_capsule, center=<3.2327 0 0>, rotation=<rot 90 0 0>}]
 dyn_R_frontLock_1_01 [{shapeType=Capsule, length=4.2211, radius=1.25, shapeName=dyn_R_frontLock_1_01_capsule, center=<4.0364 -0.111 0>, rotation=<rot 90 0 0>}]
 dyn_R_frontLock_1_03 [{shapeType=Capsule, length=2.2008, radius=1.7278, shapeName=dyn_R_frontLock_1_03_capsule, center=<2.7101 0 0>, rotation=<rot 90 0 0>}]
 dyn_R_frontLock_1_04 [{shapeType=Capsule, length=4.919, radius=1.25, shapeName=dyn_R_frontLock_1_04_capsule, center=<4.3135 0 0>, rotation=<rot 90 0 0>}]
 dyn_R_frontLock_1_06 [{
   shapeType=Capsule
   length=6.2217
   radius=1.5707
   shapeName=dyn_R_frontLock_1_06_capsule
   center=<2.5393 -1.2221 1.4981>
   rotation=<rot 90 0 0>
  }]
 dyn_R_frontLock_2_01 [{shapeType=Capsule, length=4.5377, radius=1.2, shapeName=dyn_R_frontLock_2_01_capsule, center=<3.9963 -0.145 0>, rotation=<rot 90 0 0>}]
 dyn_R_frontLock_2_03 [{shapeType=Capsule, length=2.1312, radius=1.5702, shapeName=dyn_R_frontLock_2_03_capsule, center=<2.8879 0 0>, rotation=<rot 90 0 0>}]
 dyn_R_frontLock_2_04 [{shapeType=Capsule, length=5.5631, radius=1.5068, shapeName=dyn_R_frontLock_2_04_capsule, center=<4.3551 0 0>, rotation=<rot 74.9999 0 0>}]
 dyn_R_frontLock_2_06 [{shapeType=Capsule, length=7.836, radius=2.1796, shapeName=dyn_R_frontLock_2_06_capsule, center=<3.106 0 0>, rotation=<rot 90 0 0>}]
 dyn_R_frontLock_3_01 [{shapeType=Capsule, length=4, radius=1.5, shapeName=dyn_R_frontLock_3_01_capsule, center=<3.9553 0 0>, rotation=<rot 90 0 0>}]
 dyn_R_frontLock_3_03 [{shapeType=Capsule, length=4.5477, radius=1.5, shapeName=dyn_R_frontLock_3_03_capsule, center=<3.9915 0 0>, rotation=<rot 90 0 0>}]
 dyn_R_frontLock_3_05 [{
   shapeType=Capsule
   length=4.3154
   radius=2.1906
   shapeName=dyn_R_frontLock_3_05_capsule
   center=<1.2861 0.3834 0.2032>
   rotation=<rot 79.9995 -0.0002 -0.0002>
  }]
 dyn_R_frontLock_3_06 [{shapeType=Capsule, length=6.4023, radius=1.53, shapeName=dyn_R_frontLock_3_06_capsule, center=<4.3022 0 0>, rotation=<rot 90 0 0>}]
 head [{shapeType=Capsule, length=7, radius=6.5, shapeName=head_capsule, center=<4.4565 1.6062 0>, rotation=<rot 90 0 0>}]
 neck_02 [{shapeType=Capsule, length=7.5, radius=5.5, shapeName=neck_02_capsule, center=<0 0.6447 0>, rotation=<rot 90 0 0>}]
 spine_04 [{shapeType=Capsule, length=10.7459, radius=12.5, shapeName=spine_04_capsule, center=<4.5581 -1.0778 0>, rotation=<rot 0 -90 0>}]
 spine_05 [
  {shapeType=Capsule, length=11.416, radius=9.6483, shapeName=spine_05_capsule, center=<3.9212 -1.6 0>, rotation=<rot 0 -90 0>}
  {shapeType=Capsule, length=15.2857, radius=10, shapeName=spine_05_capsule, center=<1 -0.5 0>, rotation=<rot 0 -90 0>}
 ]
 upperarm_l [{shapeType=Capsule, length=18.7912, radius=5.3164, shapeName=upperarm_r_capsule, center=<13.3542 -1.4817 0.403>, rotation=<rot -90 0 0>}]
 upperarm_r [{shapeType=Capsule, length=18.7912, radius=6.8164, shapeName=upperarm_r_capsule, center=<-11.5173 0.898 0.403>, rotation=<rot -90 0 0>}]
}
constraints [
 {
  bone1Name=base_L_backLock_1
  bone2Name=head
  swing1LimitDegrees=45
  swing1Motion=Free
  swing2LimitDegrees=45
  swing2Motion=Free
  twistLimitDegrees=45
  twistMotion=Free
 }
 {bone1Name=dyn_L_backLock_1_01, bone2Name=base_L_backLock_1, swing1LimitDegrees=10, swing2LimitDegrees=10, twistLimitDegrees=0}
 {bone1Name=dyn_L_backLock_1_03, bone2Name=dyn_L_backLock_1_01, swing1LimitDegrees=20, swing2LimitDegrees=20, twistLimitDegrees=5}
 {bone1Name=dyn_L_backLock_1_04, bone2Name=dyn_L_backLock_1_03, swing1LimitDegrees=20, swing2LimitDegrees=20, twistLimitDegrees=5}
 {
  bone1Name=base_L_backLock_2
  bone2Name=head
  swing1LimitDegrees=45
  swing1Motion=Free
  swing2LimitDegrees=45
  swing2Motion=Free
  twistLimitDegrees=45
  twistMotion=Free
 }
 {bone1Name=dyn_L_backLock_2_01, bone2Name=base_L_backLock_2, swing1LimitDegrees=10, swing2LimitDegrees=10, twistLimitDegrees=0}
 {bone1Name=dyn_L_backLock_2_03, bone2Name=dyn_L_backLock_2_01, swing1LimitDegrees=10, swing2LimitDegrees=10, twistLimitDegrees=0}
 {bone1Name=dyn_L_backLock_2_05, bone2Name=dyn_L_backLock_2_03, swing1LimitDegrees=20, swing2LimitDegrees=20, twistLimitDegrees=5}
 {bone1Name=dyn_L_backLock_2_06, bone2Name=dyn_L_backLock_2_05, swing1LimitDegrees=20, swing2LimitDegrees=20, twistLimitDegrees=5}
 {
  bone1Name=base_L_frontLock_1
  bone2Name=head
  swing1LimitDegrees=45
  swing1Motion=Free
  swing2LimitDegrees=45
  swing2Motion=Free
  twistLimitDegrees=45
  twistMotion=Free
 }
 {bone1Name=dyn_L_frontLock_1_01, bone2Name=base_L_frontLock_1, swing1LimitDegrees=10, swing2LimitDegrees=20, twistLimitDegrees=5}
 {bone1Name=dyn_L_frontLock_1_03, bone2Name=dyn_L_frontLock_1_01, swing1LimitDegrees=20, swing2LimitDegrees=20, twistLimitDegrees=5}
 {bone1Name=dyn_L_frontLock_1_05, bone2Name=dyn_L_frontLock_1_03, swing1LimitDegrees=20, swing2LimitDegrees=20, twistLimitDegrees=5}
 {bone1Name=dyn_L_frontLock_1_06, bone2Name=dyn_L_frontLock_1_05, swing1LimitDegrees=20, swing2LimitDegrees=20, twistLimitDegrees=5}
 {
  bone1Name=base_L_frontLock_2
  bone2Name=head
  swing1LimitDegrees=45
  swing1Motion=Free
  swing2LimitDegrees=45
  swing2Motion=Free
  twistLimitDegrees=45
  twistMotion=Free
 }
 {bone1Name=dyn_L_frontLock_2_01, bone2Name=base_L_frontLock_2, swing1LimitDegrees=10, swing2LimitDegrees=20, twistLimitDegrees=5}
 {bone1Name=dyn_L_frontLock_2_03, bone2Name=dyn_L_frontLock_2_01, swing1LimitDegrees=20, swing2LimitDegrees=20, twistLimitDegrees=5}
 {bone1Name=dyn_L_frontLock_2_04, bone2Name=dyn_L_frontLock_2_03, swing1LimitDegrees=20, swing2LimitDegrees=20, twistLimitDegrees=5}
 {bone1Name=dyn_L_frontLock_2_05, bone2Name=dyn_L_frontLock_2_04, swing1LimitDegrees=20, swing2LimitDegrees=20, twistLimitDegrees=5}
 {
  bone1Name=base_L_frontLock_3
  bone2Name=head
  swing1LimitDegrees=45
  swing1Motion=Free
  swing2LimitDegrees=45
  swing2Motion=Free
  twistLimitDegrees=45
  twistMotion=Free
 }
 {bone1Name=dyn_L_frontLock_3_01, bone2Name=base_L_frontLock_3, swing1LimitDegrees=10, swing2LimitDegrees=20, twistLimitDegrees=5}
 {bone1Name=dyn_L_frontLock_3_03, bone2Name=dyn_L_frontLock_3_01, swing1LimitDegrees=20, swing2LimitDegrees=20, twistLimitDegrees=5}
 {bone1Name=dyn_L_frontLock_3_05, bone2Name=dyn_L_frontLock_3_03, swing1LimitDegrees=20, swing2LimitDegrees=20, twistLimitDegrees=5}
 {bone1Name=dyn_L_frontLock_3_06, bone2Name=dyn_L_frontLock_3_05, swing1LimitDegrees=20, swing2LimitDegrees=20, twistLimitDegrees=5}
 {
  bone1Name=base_L_frontLock_4
  bone2Name=head
  swing1LimitDegrees=45
  swing1Motion=Free
  swing2LimitDegrees=45
  swing2Motion=Free
  twistLimitDegrees=45
  twistMotion=Free
 }
 {bone1Name=dyn_L_frontLock_4_01, bone2Name=base_L_frontLock_4, swing1LimitDegrees=10, swing2LimitDegrees=20, twistLimitDegrees=5}
 {bone1Name=dyn_L_frontLock_4_03, bone2Name=dyn_L_frontLock_4_01, swing1LimitDegrees=20, swing2LimitDegrees=20, twistLimitDegrees=5}
 {bone1Name=dyn_L_frontLock_4_05, bone2Name=dyn_L_frontLock_4_03, swing1LimitDegrees=20, swing2LimitDegrees=20, twistLimitDegrees=5}
 {bone1Name=dyn_L_frontLock_4_06, bone2Name=dyn_L_frontLock_4_05, swing1LimitDegrees=20, swing2LimitDegrees=20, twistLimitDegrees=5}
 {
  bone1Name=base_M_backLock_1
  bone2Name=head
  swing1LimitDegrees=45
  swing1Motion=Free
  swing2LimitDegrees=45
  swing2Motion=Free
  twistLimitDegrees=45
  twistMotion=Free
 }
 {bone1Name=dyn_M_backLock_1_01, bone2Name=base_M_backLock_1, swing1LimitDegrees=20, swing2LimitDegrees=20, twistLimitDegrees=5}
 {bone1Name=dyn_M_backLock_1_03, bone2Name=dyn_M_backLock_1_01, swing1LimitDegrees=20, swing2LimitDegrees=20, twistLimitDegrees=5}
 {bone1Name=dyn_M_backLock_1_05, bone2Name=dyn_M_backLock_1_03, swing1LimitDegrees=20, swing2LimitDegrees=20, twistLimitDegrees=5}
 {
  bone1Name=base_M_backLock_2
  bone2Name=head
  swing1LimitDegrees=45
  swing1Motion=Free
  swing2LimitDegrees=45
  swing2Motion=Free
  twistLimitDegrees=45
  twistMotion=Free
 }
 {bone1Name=dyn_M_backLock_2_01, bone2Name=base_M_backLock_2, swing1LimitDegrees=5, swing2LimitDegrees=10, twistLimitDegrees=5}
 {bone1Name=dyn_M_backLock_2_03, bone2Name=dyn_M_backLock_2_01, swing1LimitDegrees=15, swing2LimitDegrees=10, twistLimitDegrees=5}
 {bone1Name=dyn_M_backLock_2_05, bone2Name=dyn_M_backLock_2_03, swing1LimitDegrees=20, swing2LimitDegrees=20, twistLimitDegrees=5}
 {
  bone1Name=base_M_backLock_3
  bone2Name=head
  swing1LimitDegrees=45
  swing1Motion=Free
  swing2LimitDegrees=45
  swing2Motion=Free
  twistLimitDegrees=45
  twistMotion=Free
 }
 {bone1Name=dyn_M_backLock_3_01, bone2Name=base_M_backLock_3, swing1LimitDegrees=20, swing2LimitDegrees=20, twistLimitDegrees=5}
 {bone1Name=dyn_M_backLock_3_03, bone2Name=dyn_M_backLock_3_01, swing1LimitDegrees=20, swing2LimitDegrees=20, twistLimitDegrees=5}
 {bone1Name=dyn_M_backLock_3_05, bone2Name=dyn_M_backLock_3_03, swing1LimitDegrees=20, swing2LimitDegrees=20, twistLimitDegrees=5}
 {
  bone1Name=base_R_backLock_1
  bone2Name=head
  swing1LimitDegrees=45
  swing1Motion=Free
  swing2LimitDegrees=45
  swing2Motion=Free
  twistLimitDegrees=45
  twistMotion=Free
 }
 {bone1Name=dyn_R_backLock_1_01, bone2Name=base_R_backLock_1, swing1LimitDegrees=20, swing2LimitDegrees=20, twistLimitDegrees=5}
 {bone1Name=dyn_R_backLock_1_03, bone2Name=dyn_R_backLock_1_01, swing1LimitDegrees=20, swing2LimitDegrees=20, twistLimitDegrees=5}
 {bone1Name=dyn_R_backLock_1_05, bone2Name=dyn_R_backLock_1_03, swing1LimitDegrees=30, swing2LimitDegrees=20, twistLimitDegrees=5}
 {
  bone1Name=base_R_backLock_2
  bone2Name=head
  swing1LimitDegrees=45
  swing1Motion=Free
  swing2LimitDegrees=45
  swing2Motion=Free
  twistLimitDegrees=45
  twistMotion=Free
 }
 {bone1Name=dyn_R_backLock_2_01, bone2Name=base_R_backLock_2, swing1LimitDegrees=10, swing2LimitDegrees=10, twistLimitDegrees=5}
 {bone1Name=dyn_R_backLock_2_03, bone2Name=dyn_R_backLock_2_01, swing1LimitDegrees=30, swing2LimitDegrees=20, twistLimitDegrees=5}
 {bone1Name=dyn_R_backLock_2_05, bone2Name=dyn_R_backLock_2_03, swing1LimitDegrees=30, swing2LimitDegrees=20, twistLimitDegrees=5}
 {
  bone1Name=base_R_frontLock_1
  bone2Name=head
  swing1LimitDegrees=45
  swing1Motion=Free
  swing2LimitDegrees=45
  swing2Motion=Free
  twistLimitDegrees=45
  twistMotion=Free
 }
 {bone1Name=dyn_R_frontLock_1_01, bone2Name=base_R_frontLock_1, swing1LimitDegrees=10, swing2LimitDegrees=20, twistLimitDegrees=5}
 {bone1Name=dyn_R_frontLock_1_03, bone2Name=dyn_R_frontLock_1_01, swing1LimitDegrees=20, swing2LimitDegrees=20, twistLimitDegrees=5}
 {bone1Name=dyn_R_frontLock_1_04, bone2Name=dyn_R_frontLock_1_03, swing1LimitDegrees=20, swing2LimitDegrees=20, twistLimitDegrees=5}
 {bone1Name=dyn_R_frontLock_1_06, bone2Name=dyn_R_frontLock_1_04, swing1LimitDegrees=20, swing2LimitDegrees=20, twistLimitDegrees=5}
 {
  bone1Name=base_R_frontLock_2
  bone2Name=head
  swing1LimitDegrees=45
  swing1Motion=Free
  swing2LimitDegrees=45
  swing2Motion=Free
  twistLimitDegrees=45
  twistMotion=Free
 }
 {bone1Name=dyn_R_frontLock_2_01, bone2Name=base_R_frontLock_2, swing1LimitDegrees=10, swing2LimitDegrees=10, twistLimitDegrees=2}
 {bone1Name=dyn_R_frontLock_2_03, bone2Name=dyn_R_frontLock_2_01, swing1LimitDegrees=15, swing2LimitDegrees=15, twistLimitDegrees=2}
 {bone1Name=dyn_R_frontLock_2_04, bone2Name=dyn_R_frontLock_2_03, swing1LimitDegrees=15, swing2LimitDegrees=15, twistLimitDegrees=0}
 {bone1Name=dyn_R_frontLock_2_06, bone2Name=dyn_R_frontLock_2_04, swing1LimitDegrees=20, swing2LimitDegrees=20, twistLimitDegrees=0}
 {
  bone1Name=base_R_frontLock_3
  bone2Name=head
  swing1LimitDegrees=45
  swing1Motion=Free
  swing2LimitDegrees=45
  swing2Motion=Free
  twistLimitDegrees=45
  twistMotion=Free
 }
 {bone1Name=dyn_R_frontLock_3_01, bone2Name=base_R_frontLock_3, swing1LimitDegrees=10, swing2LimitDegrees=20, twistLimitDegrees=5}
 {bone1Name=dyn_R_frontLock_3_03, bone2Name=dyn_R_frontLock_3_01, swing1LimitDegrees=20, swing2LimitDegrees=20, twistLimitDegrees=5}
 {bone1Name=dyn_R_frontLock_3_05, bone2Name=dyn_R_frontLock_3_03, swing1LimitDegrees=20, swing2LimitDegrees=20, twistLimitDegrees=5}
 {bone1Name=dyn_R_frontLock_3_06, bone2Name=dyn_R_frontLock_3_05, swing1LimitDegrees=20, swing2LimitDegrees=20, twistLimitDegrees=5}
 {
  bone1Name=dyn_R_frontLock_3_05
  bone2Name=dyn_R_frontLock_1_04
  swing1LimitDegrees=45
  swing1Motion=Free
  swing2LimitDegrees=45
  swing2Motion=Free
  twistLimitDegrees=45
  twistMotion=Free
 }
 {
  bone1Name=dyn_L_frontLock_3_06
  bone2Name=dyn_L_frontLock_1_06
  swing1LimitDegrees=45
  swing1Motion=Free
  swing2LimitDegrees=45
  swing2Motion=Free
  twistLimitDegrees=45
  twistMotion=Free
 }
 {
  bone1Name=dyn_R_frontLock_2_06
  bone2Name=dyn_R_frontLock_1_06
  swing1LimitDegrees=45
  swing1Motion=Free
  swing2LimitDegrees=45
  swing2Motion=Free
  twistLimitDegrees=45
  twistMotion=Free
 }
]
