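# Post-process AnimBP for the EssayViewMythHat face accessory, driving its hat and hair lock bones.
# Copied parent and body-mesh poses feed a Control Rig, then one of three RigidBody branches: backbling or front-end skydive (with bone constraints), EssayPier character, or default.
# Bone offsets adjust the latter two.
# The event graph detects backbling and character, varying gravity for skydive states and sim-space settings, which only the Pier simulation reads.
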AnimBlueprint /BRCosmetics/Characters/Player/Male/Medium/Bodies/M_MED_EssayViewMythHat/Meshes/Parts/M_MED_EssayViewMythHat_FaceAcc_AnimBP
target_skeleton=&/M_MED_EssayViewMythHat_FaceAcc_Skeleton
graphs [
 AnimGraph {nodes [
   "Constraint - base_M_backLock_1 (PPPP)"!(
    ComponentPose
    Alpha=1
    ConstraintWeights_0=0.8674
    ConstraintWeights_1:float=0.0689
    ConstraintWeights_2:float=0.047
    ConstraintWeights_3:float=0.0167
   ) -> (Pose) {
    alphaBoolBlend {blendOption=Linear}
    boneToModify {boneName=base_M_backLock_1}
    constraintSetup [{targetBone {boneName=spine_05}}, {targetBone {boneName=clavicle_r}}, {targetBone {boneName=neck_01}}, {targetBone {boneName=clavicle_l}}]
    constraintWeights [1,1,1,1]
   }
   "Constraint - base_M_backLock_2 (PPP)"!(ComponentPose, Alpha=1, ConstraintWeights_0=0.85, ConstraintWeights_1:float=0.1169, ConstraintWeights_2:float=0.0417) -> (Pose) {
    alphaBoolBlend {blendOption=Linear}
    boneToModify {boneName=base_M_backLock_2}
    constraintSetup [{targetBone {boneName=spine_05}}, {targetBone {boneName=clavicle_r}}, {targetBone {boneName=neck_01}}]
    constraintWeights [1,1,1]
   }
   "Constraint - base_M_backLock_3 (PPPPP)"!(
    ComponentPose
    Alpha=1
    ConstraintWeights_0=0.9376
    ConstraintWeights_1:float=0.0516
    ConstraintWeights_2:float=0.0031
    ConstraintWeights_3:float=0.0029
    ConstraintWeights_4:float=0.0022
   ) -> (Pose) {
    alphaBoolBlend {blendOption=Linear}
    boneToModify {boneName=base_M_backLock_3}
    constraintSetup [
     {targetBone {boneName=spine_05}}
     {targetBone {boneName=clavicle_l}}
     {targetBone {boneName=neck_01}}
     {targetBone {boneName=spine_03}}
     {targetBone {boneName=spine_04}}
    ]
    constraintWeights [1,1,1,1,1]
   }
   "Constraint - base_L_backLock_2 (PPPPP)"!(
    ComponentPose
    Alpha=1
    ConstraintWeights_0=0.8745
    ConstraintWeights_1:float=0.1094
    ConstraintWeights_2:float=0.0075
    ConstraintWeights_3:float=0.003
    ConstraintWeights_4:float=0.0026
   ) -> (Pose) {
    alphaBoolBlend {blendOption=Linear}
    boneToModify {boneName=base_L_backLock_2}
    constraintSetup [
     {targetBone {boneName=spine_05}}
     {targetBone {boneName=clavicle_l}}
     {targetBone {boneName=neck_01}}
     {targetBone {boneName=spine_03}}
     {targetBone {boneName=spine_04}}
    ]
    constraintWeights [1,1,1,1,1]
   }
   "Constraint - base_R_backLock_1 (PPP)"!(ComponentPose, Alpha=1, ConstraintWeights_0=0.6647, ConstraintWeights_1:float=0.2951, ConstraintWeights_2:float=0.0402) -> (Pose) {
    alphaBoolBlend {blendOption=Linear}
    boneToModify {boneName=base_R_backLock_1}
    constraintSetup [{targetBone {boneName=spine_05}}, {targetBone {boneName=clavicle_r}}, {targetBone {boneName=neck_01}}]
    constraintWeights [1,1,1]
   }
   "Constraint - base_R_backLock_2 (PPP)"!(ComponentPose, Alpha=1, ConstraintWeights_0=0.6934, ConstraintWeights_1:float=0.2714, ConstraintWeights_2:float=0.0353) -> (Pose) {
    alphaBoolBlend {blendOption=Linear}
    boneToModify {boneName=base_R_backLock_2}
    constraintSetup [{targetBone {boneName=spine_05}}, {targetBone {boneName=clavicle_r}}, {targetBone {boneName=neck_01}}]
    constraintWeights [1,1,1]
   }
   "Transform (Modify) Bone - Bone: base_L_frontLock_3"!(ComponentPose, bAlphaBoolEnabled:bool=True, Translation="0,2,0", Rotation="0,0,5", Scale="1,1,1") -> (Pose) {
    alphaInputType=Bool
    rotationMode=BMM_Additive
    rotationSpace=BCS_BoneSpace
    scaleMode=BMM_Additive
    scaleSpace=BCS_BoneSpace
    translationMode=BMM_Additive
    translationSpace=BCS_BoneSpace
    alphaBoolBlend {blendInTime=1, blendOption=Linear, blendOutTime=1}
    boneToModify {boneName=base_L_frontLock_3}
    rotation {roll=5}
    translation {y=2}
   }
   n1 = "Transform (Modify) Bone - Bone: base_R_frontLock_2"(Alpha=1, Translation="0,0,0", Rotation="0,-5,0", Scale="1,1,1") {
    rotationMode=BMM_Additive
    rotationSpace=BCS_BoneSpace
    scaleMode=BMM_Additive
    scaleSpace=BCS_BoneSpace
    translationMode=BMM_Additive
    translationSpace=BCS_BoneSpace
    alphaBoolBlend {blendOption=Linear}
    boneToModify {boneName=base_R_frontLock_2}
    rotation {yaw=-5}
   }
   "Transform (Modify) Bone - Bone: base_R_frontLock_1"(ComponentPose=@n1, Alpha=1, Translation="0,0,0", Rotation="0,-5,0", Scale="1,1,1") {
    rotationMode=BMM_Additive
    rotationSpace=BCS_BoneSpace
    scaleMode=BMM_Additive
    scaleSpace=BCS_BoneSpace
    translationMode=BMM_Additive
    translationSpace=BCS_BoneSpace
    alphaBoolBlend {blendOption=Linear}
    boneToModify {boneName=base_R_frontLock_1}
    rotation {yaw=-5}
   }
   n2 = "Transform (Modify) Bone - Bone: base_hatback"!(
    ComponentPose=Local To Component(LocalPose=Use cached pose 'Control Rig Pose'())
    bAlphaBoolEnabled:bool=True
    Translation="0,0,0"
    Rotation="0,0,30"
    Scale="1,1,1"
   ) -> (Pose) {
    alphaInputType=Bool
    rotationMode=BMM_Additive
    rotationSpace=BCS_BoneSpace
    translationMode=BMM_Additive
    translationSpace=BCS_BoneSpace
    alphaBoolBlend {blendInTime=1, blendOption=Linear, blendOutTime=1}
    boneToModify {boneName=base_hatback}
    rotation {roll=30}
   }
   n3 = "Transform (Modify) Bone - Bone: base_M_backLock_2"!(ComponentPose=@n2, bAlphaBoolEnabled:bool=True, Translation="0,0,0", Rotation="-0.000,29.321,0.000", Scale="1,1,1") -> (Pose) {
    alphaInputType=Bool
    rotationMode=BMM_Additive
    rotationSpace=BCS_BoneSpace
    translationMode=BMM_Additive
    translationSpace=BCS_BoneSpace
    alphaBoolBlend {blendInTime=1, blendOption=Linear, blendOutTime=1}
    boneToModify {boneName=base_M_backLock_2}
    rotation {yaw=29.321}
   }
   n4 = "Transform (Modify) Bone - Bone: base_M_backLock_1"!(ComponentPose=@n3, bAlphaBoolEnabled:bool=True, Translation="0,0,0", Rotation="24.813,22.156,-0.000", Scale="1,1,1") -> (Pose) {
    alphaInputType=Bool
    rotationMode=BMM_Additive
    rotationSpace=BCS_BoneSpace
    translationMode=BMM_Additive
    translationSpace=BCS_BoneSpace
    alphaBoolBlend {blendInTime=1, blendOption=Linear, blendOutTime=1}
    boneToModify {boneName=base_M_backLock_1}
    rotation {pitch=24.813, yaw=22.156}
   }
   n5 = "Transform (Modify) Bone - Bone: base_R_backLock_2"!(ComponentPose=@n4, bAlphaBoolEnabled:bool=True, Translation="-0.580,1.016,0.064", Rotation="6.392,29.287,-3.304", Scale="1,1,1") -> (Pose) {
    alphaInputType=Bool
    rotationMode=BMM_Additive
    rotationSpace=BCS_BoneSpace
    translationMode=BMM_Additive
    translationSpace=BCS_BoneSpace
    alphaBoolBlend {blendInTime=1, blendOption=Linear, blendOutTime=1}
    boneToModify {boneName=base_R_backLock_2}
    rotation=<rot 6.392 29.287 -3.304>
    translation=<-0.58 1.016 0.064>
   }
   n6 = "Transform (Modify) Bone - Bone: base_R_backLock_1"!(ComponentPose=@n5, bAlphaBoolEnabled:bool=True, Translation="0,0,0", Rotation="0.000,27.685,-0.000", Scale="1,1,1") -> (Pose) {
    alphaInputType=Bool
    rotationMode=BMM_Additive
    rotationSpace=BCS_BoneSpace
    translationMode=BMM_Additive
    translationSpace=BCS_BoneSpace
    alphaBoolBlend {blendInTime=1, blendOption=Linear, blendOutTime=1}
    boneToModify {boneName=base_R_backLock_1}
    rotation {yaw=27.685}
   }
   n7 = "Transform (Modify) Bone - Bone: base_M_backLock_2"(ComponentPose=@n6, bAlphaBoolEnabled=True, Translation="1.429,-3.653,-0.333", Rotation="4.802,73.086,-1.284", Scale="1,1,1") {
    alphaInputType=Bool
    rotationMode=BMM_Additive
    rotationSpace=BCS_BoneSpace
    translationMode=BMM_Additive
    translationSpace=BCS_BoneSpace
    alphaBoolBlend {blendInTime=1, blendOption=Linear, blendOutTime=1}
    boneToModify {boneName=base_M_backLock_2}
    rotation=<rot 4.802 73.086 -1.284>
    translation=<1.429 -3.653 -0.333>
   }
   n8 = "Transform (Modify) Bone - Bone: base_M_backLock_1"(ComponentPose=@n7, bAlphaBoolEnabled=True, Translation="0,0,0", Rotation="27.869,0.000,0.000", Scale="1,1,1") {
    alphaInputType=Bool
    rotationMode=BMM_Additive
    rotationSpace=BCS_BoneSpace
    translationMode=BMM_Additive
    translationSpace=BCS_BoneSpace
    alphaBoolBlend {blendInTime=1, blendOption=Linear, blendOutTime=1}
    boneToModify {boneName=base_M_backLock_1}
    rotation=<rot 27.869 0 0>
   }
   n9 = "Transform (Modify) Bone - Bone: base_R_backLock_2"(ComponentPose=@n8, bAlphaBoolEnabled=True, Translation="0,0,0", Rotation="1.001,-9.593,-0.206", Scale="1,1,1") {
    alphaInputType=Bool
    rotationMode=BMM_Additive
    rotationSpace=BCS_BoneSpace
    translationMode=BMM_Additive
    translationSpace=BCS_BoneSpace
    alphaBoolBlend {blendInTime=1, blendOption=Linear, blendOutTime=1}
    boneToModify {boneName=base_R_backLock_2}
    rotation=<rot 1.001 -9.593 -0.206>
   }
   n10 = "Transform (Modify) Bone - Bone: base_R_backLock_1"(ComponentPose=@n9, bAlphaBoolEnabled=True, Translation="0,0,0", Rotation="-6.353,-22.227,-0.091", Scale="1,1,1") {
    alphaInputType=Bool
    rotationMode=BMM_Additive
    rotationSpace=BCS_BoneSpace
    translationMode=BMM_Additive
    translationSpace=BCS_BoneSpace
    alphaBoolBlend {blendInTime=1, blendOption=Linear, blendOutTime=1}
    boneToModify {boneName=base_R_backLock_1}
    rotation=<rot -6.353 -22.227 -0.091>
   }
   n11 = RigidBody!(
    ComponentPose=@n10
    Alpha
    OverrideWorldGravity:Vector=Get GravityOverride()
    ExternalForce
    SimSpaceSettings:Sim Space Settings Structure=Get Sim Space Settings() -> (Sim Space Settings:Sim Space Settings Structure)
   ) -> (Pose) {
    bClampLinearTranslationLimitToRefPose=true
    bOverrideWorldGravity=true
    overridePhysicsAsset=&/M_MED_EssayViewMyth_Hat_FaceAcc_Pier_Back_Physics
    simulationSpace=BaseBoneSpace
    alphaBoolBlend {blendOption=Linear}
    baseBoneRef {boneName=spine_05}
    simSpaceSettings {
     dampingAlpha=0
     maxAngularAcceleration=16
     maxAngularVelocity=8
     maxLinearAcceleration=800
     maxLinearVelocity=600
     externalLinearDragV=<2 0.2 0.2>
    }
   }
   n12 = "Transform (Modify) Bone - Bone: dyn_L_frontLock_3_06"!(ComponentPose=@n11, bAlphaBoolEnabled:bool=True, Translation="0,0,0", Rotation="0,15,0", Scale="1,1,1") -> (Pose) {
    alphaInputType=Bool
    rotationMode=BMM_Additive
    rotationSpace=BCS_BoneSpace
    scaleMode=BMM_Additive
    scaleSpace=BCS_BoneSpace
    translationMode=BMM_Additive
    translationSpace=BCS_BoneSpace
    alphaBoolBlend {blendInTime=1, blendOption=Linear, blendOutTime=1}
    boneToModify {boneName=dyn_L_frontLock_3_06}
    rotation {yaw=15}
   }
   n13 = Component To Local(ComponentPose=@n12)
   n14 = RigidBody!(
    ComponentPose=Local To Component(LocalPose=Use cached pose 'Control Rig Pose'())
    Alpha
    OverrideWorldGravity=Get GravityOverride()
    ExternalForce
   ) -> (Pose) {
    bClampLinearTranslationLimitToRefPose=true
    bOverrideWorldGravity=true
    overridePhysicsAsset=&/M_MED_EssayViewMyth_Hat_FaceAcc_Back_Physics
    simulationSpace=BaseBoneSpace
    alphaBoolBlend {blendOption=Linear}
    baseBoneRef {boneName=spine_05}
    simSpaceSettings {
     dampingAlpha=0
     maxAngularAcceleration=16
     maxAngularVelocity=8
     maxLinearAcceleration=800
     maxLinearVelocity=600
     worldAlpha=1
     externalLinearDragV=<2 0.2 0.2>
    }
   }
   n15 = "Transform (Modify) Bone - Bone: base_L_backLock_1"!(ComponentPose=@n14, bAlphaBoolEnabled:bool=True, Translation="-0.2,0,0", Rotation="0,0,0.01", Scale="1,1,1") -> (Pose) {
    alphaInputType=Bool
    rotationMode=BMM_Additive
    rotationSpace=BCS_BoneSpace
    scaleMode=BMM_Additive
    scaleSpace=BCS_BoneSpace
    translationMode=BMM_Additive
    translationSpace=BCS_BoneSpace
    alphaBoolBlend {blendInTime=1, blendOption=Linear, blendOutTime=1}
    boneToModify {boneName=base_L_backLock_1}
    rotation {roll=0.01}
    translation {x=-0.2}
   }
   n16 = "Transform (Modify) Bone - Bone: base_R_frontLock_3"!(ComponentPose=@n15, bAlphaBoolEnabled:bool=True, Translation="0,0,0", Rotation="0,5,0.001", Scale="1,1,1") -> (Pose) {
    alphaInputType=Bool
    rotationMode=BMM_Additive
    rotationSpace=BCS_BoneSpace
    scaleMode=BMM_Additive
    scaleSpace=BCS_BoneSpace
    translationMode=BMM_Additive
    translationSpace=BCS_BoneSpace
    alphaBoolBlend {blendInTime=1, blendOption=Linear, blendOutTime=1}
    boneToModify {boneName=base_R_frontLock_3}
    rotation {roll=0.001, yaw=5}
   }
   n17 = Component To Local(ComponentPose=@n16)
   n18 = Get isEssayPier() -> (isEssayPier:bool)
   n19 = "Constraint - dyn_L_backLock_2_01 (PPPPP)"!(
    ComponentPose=Local To Component(LocalPose=Use cached pose 'Control Rig Pose'())
    Alpha=1
    ConstraintWeights_0=0
    ConstraintWeights_1:float=0
    ConstraintWeights_2:float=0.34
    ConstraintWeights_3:float=0.33
    ConstraintWeights_4:float=0.33
   ) -> (Pose) {
    alphaBoolBlend {blendOption=Linear}
    boneToModify {boneName=dyn_L_backLock_2_01}
    constraintSetup [
     {targetBone {boneName=spine_05}}
     {targetBone {boneName=clavicle_l}}
     {targetBone {boneName=neck_01}}
     {targetBone {boneName=head}}
     {targetBone {boneName=neck_02}}
    ]
    constraintWeights [1,1,1,1,1]
   }
   n20 = "Constraint - dyn_L_backLock_2_03 (PPPPP)"!(
    ComponentPose=@n19
    Alpha=1
    ConstraintWeights_0=0.33
    ConstraintWeights_1:float=0
    ConstraintWeights_2:float=0.33
    ConstraintWeights_3:float=0.34
    ConstraintWeights_4:float=0
   ) -> (Pose) {
    alphaBoolBlend {blendOption=Linear}
    boneToModify {boneName=dyn_L_backLock_2_03}
    constraintSetup [
     {targetBone {boneName=spine_05}}
     {targetBone {boneName=clavicle_l}}
     {targetBone {boneName=neck_01}}
     {targetBone {boneName=head}}
     {targetBone {boneName=spine_04}}
    ]
    constraintWeights [1,1,1,1,1]
   }
   n21 = "Constraint - dyn_L_backLock_2_05 (PPPPP)"!(
    ComponentPose=@n20
    Alpha=1
    ConstraintWeights_0=0.7
    ConstraintWeights_1:float=0
    ConstraintWeights_2:float=0.3
    ConstraintWeights_3:float=0
    ConstraintWeights_4:float=0
   ) -> (Pose) {
    alphaBoolBlend {blendOption=Linear}
    boneToModify {boneName=dyn_L_backLock_2_05}
    constraintSetup [
     {targetBone {boneName=spine_05}}
     {targetBone {boneName=clavicle_l}}
     {targetBone {boneName=neck_01}}
     {targetBone {boneName=spine_03}}
     {targetBone {boneName=spine_04}}
    ]
    constraintWeights [1,1,1,1,1]
   }
   n22 = "Constraint - dyn_L_backLock_2_06 (PPPPP)"!(
    ComponentPose=@n21
    Alpha=1
    ConstraintWeights_0=0.5
    ConstraintWeights_1:float=0.5
    ConstraintWeights_2:float=0
    ConstraintWeights_3:float=0
    ConstraintWeights_4:float=0
   ) -> (Pose) {
    alphaBoolBlend {blendOption=Linear}
    boneToModify {boneName=dyn_L_backLock_2_06}
    constraintSetup [
     {targetBone {boneName=spine_05}}
     {targetBone {boneName=clavicle_l}}
     {targetBone {boneName=spine_03}}
     {targetBone {boneName=spine_04}}
     {targetBone {boneName=neck_01}}
    ]
    constraintWeights [1,1,1,1,1]
   }
   n23 = "Constraint - dyn_M_backLock_1_01 (PPPP)"!(
    ComponentPose=@n22
    Alpha=1
    ConstraintWeights_0=0
    ConstraintWeights_1:float=0.1
    ConstraintWeights_2:float=0
    ConstraintWeights_3:float=0.9
   ) -> (Pose) {
    alphaBoolBlend {blendOption=Linear}
    boneToModify {boneName=dyn_M_backLock_1_01}
    constraintSetup [{targetBone {boneName=spine_05}}, {targetBone {boneName=neck_02}}, {targetBone {boneName=neck_01}}, {targetBone {boneName=head}}]
    constraintWeights [1,1,1,1]
   }
   n24 = "Constraint - dyn_M_backLock_1_03 (PPPP)"!(
    ComponentPose=@n23
    Alpha=1
    ConstraintWeights_0=0.25
    ConstraintWeights_1:float=0.25
    ConstraintWeights_2:float=0.25
    ConstraintWeights_3:float=0.25
   ) -> (Pose) {
    alphaBoolBlend {blendOption=Linear}
    boneToModify {boneName=dyn_M_backLock_1_03}
    constraintSetup [{targetBone {boneName=spine_05}}, {targetBone {boneName=head}}, {targetBone {boneName=neck_01}}, {targetBone {boneName=neck_02}}]
    constraintWeights [1,1,1,1]
   }
   n25 = "Constraint - dyn_M_backLock_1_05 (PPPPPP)"!(
    ComponentPose=@n24
    Alpha=1
    ConstraintWeights_0=0.8
    ConstraintWeights_1:float=0
    ConstraintWeights_2:float=0
    ConstraintWeights_3:float=0
    ConstraintWeights_4:float=0
    ConstraintWeights_5:float=0.2
   ) -> (Pose) {
    alphaBoolBlend {blendOption=Linear}
    boneToModify {boneName=dyn_M_backLock_1_05}
    constraintSetup [
     {targetBone {boneName=spine_05}}
     {targetBone {boneName=clavicle_r}}
     {targetBone {boneName=spine_03}}
     {targetBone {boneName=spine_04}}
     {targetBone {boneName=clavicle_l}}
     {targetBone {boneName=neck_01}}
    ]
    constraintWeights [1,1,1,1,1,1]
   }
   n26 = "Constraint - dyn_M_backLock_2_01 (PPP)"!(ComponentPose=@n25, Alpha=1, ConstraintWeights_0=0.6, ConstraintWeights_1:float=0.2, ConstraintWeights_2:float=0.2) -> (Pose) {
    alphaBoolBlend {blendOption=Linear}
    boneToModify {boneName=dyn_M_backLock_2_01}
    constraintSetup [{targetBone {boneName=spine_05}}, {targetBone {boneName=neck_02}}, {targetBone {boneName=head}}]
    constraintWeights [1,1,1]
   }
   n27 = "Constraint - dyn_M_backLock_2_03 (PPPP)"!(
    ComponentPose=@n26
    Alpha=1
    ConstraintWeights_0=0.8
    ConstraintWeights_1:float=0
    ConstraintWeights_2:float=0.1
    ConstraintWeights_3:float=0.1
   ) -> (Pose) {
    alphaBoolBlend {blendOption=Linear}
    boneToModify {boneName=dyn_M_backLock_2_03}
    constraintSetup [{targetBone {boneName=spine_05}}, {targetBone {boneName=head}}, {targetBone {boneName=neck_01}}, {targetBone {boneName=neck_02}}]
    constraintWeights [1,1,1,1]
   }
   n28 = "Constraint - dyn_M_backLock_2_05 (PPPPPP)"!(
    ComponentPose=@n27
    Alpha=1
    ConstraintWeights_0=0.8
    ConstraintWeights_1:float=0
    ConstraintWeights_2:float=0
    ConstraintWeights_3:float=0
    ConstraintWeights_4:float=0
    ConstraintWeights_5:float=0.2
   ) -> (Pose) {
    alphaBoolBlend {blendOption=Linear}
    boneToModify {boneName=dyn_M_backLock_2_05}
    constraintSetup [
     {targetBone {boneName=spine_05}}
     {targetBone {boneName=clavicle_r}}
     {targetBone {boneName=spine_03}}
     {targetBone {boneName=spine_04}}
     {targetBone {boneName=clavicle_l}}
     {targetBone {boneName=neck_01}}
    ]
    constraintWeights [1,1,1,1,1,1]
   }
   n29 = "Constraint - dyn_M_backLock_2_06 (PPPPP)"!(
    ComponentPose=@n28
    Alpha=1
    ConstraintWeights_0=0.9
    ConstraintWeights_1:float=0
    ConstraintWeights_2:float=0
    ConstraintWeights_3:float=0
    ConstraintWeights_4:float=0.1
   ) -> (Pose) {
    alphaBoolBlend {blendOption=Linear}
    boneToModify {boneName=dyn_M_backLock_2_06}
    constraintSetup [
     {targetBone {boneName=spine_05}}
     {targetBone {boneName=clavicle_r}}
     {targetBone {boneName=spine_03}}
     {targetBone {boneName=spine_04}}
     {targetBone {boneName=neck_01}}
    ]
    constraintWeights [1,1,1,1,1]
   }
   n30 = "Constraint - dyn_M_backLock_3_01 (PPPPP)"!(
    ComponentPose=@n29
    Alpha=1
    ConstraintWeights_0=0.3
    ConstraintWeights_1:float=0.1
    ConstraintWeights_2:float=0.1
    ConstraintWeights_3:float=0.5
    ConstraintWeights_4:float=0
   ) -> (Pose) {
    alphaBoolBlend {blendOption=Linear}
    boneToModify {boneName=dyn_M_backLock_3_01}
    constraintSetup [
     {targetBone {boneName=spine_05}}
     {targetBone {boneName=neck_02}}
     {targetBone {boneName=neck_01}}
     {targetBone {boneName=head}}
     {targetBone {boneName=spine_04}}
    ]
    constraintWeights [1,1,1,1,1]
   }
   n31 = "Constraint - dyn_M_backLock_3_03 (PPPP)"!(
    ComponentPose=@n30
    Alpha=1
    ConstraintWeights_0=0.5
    ConstraintWeights_1:float=0.25
    ConstraintWeights_2:float=0.25
    ConstraintWeights_3:float=0
   ) -> (Pose) {
    alphaBoolBlend {blendOption=Linear}
    boneToModify {boneName=dyn_M_backLock_3_03}
    constraintSetup [{targetBone {boneName=spine_05}}, {targetBone {boneName=neck_01}}, {targetBone {boneName=neck_02}}, {targetBone {boneName=head}}]
    constraintWeights [1,0,1,1]
   }
   n32 = "Constraint - dyn_M_backLock_3_05 (PPPPPP)"!(
    ComponentPose=@n31
    Alpha=1
    ConstraintWeights_0=0.9
    ConstraintWeights_1:float=0
    ConstraintWeights_2:float=0
    ConstraintWeights_3:float=0
    ConstraintWeights_4:float=0.1
    ConstraintWeights_5:float=0
   ) -> (Pose) {
    alphaBoolBlend {blendOption=Linear}
    boneToModify {boneName=dyn_M_backLock_3_05}
    constraintSetup [
     {targetBone {boneName=spine_05}}
     {targetBone {boneName=clavicle_l}}
     {targetBone {boneName=spine_03}}
     {targetBone {boneName=clavicle_r}}
     {targetBone {boneName=neck_01}}
     {targetBone {boneName=spine_04}}
    ]
    constraintWeights [1,1,1,1,1,1]
   }
   n33 = "Constraint - dyn_M_backLock_3_06 (PPPPP)"!(
    ComponentPose=@n32
    Alpha=1
    ConstraintWeights_0=0.9
    ConstraintWeights_1:float=0
    ConstraintWeights_2:float=0
    ConstraintWeights_3:float=0
    ConstraintWeights_4:float=0.1
   ) -> (Pose) {
    alphaBoolBlend {blendOption=Linear}
    boneToModify {boneName=dyn_M_backLock_3_06}
    constraintSetup [
     {targetBone {boneName=spine_05}}
     {targetBone {boneName=clavicle_l}}
     {targetBone {boneName=spine_03}}
     {targetBone {boneName=spine_04}}
     {targetBone {boneName=neck_01}}
    ]
    constraintWeights [1,1,1,1,1]
   }
   n34 = "Constraint - dyn_R_backLock_1_01 (PPP)"!(ComponentPose=@n33, Alpha=1, ConstraintWeights_0=0.8, ConstraintWeights_1:float=0.1, ConstraintWeights_2:float=0.1) -> (Pose) {
    alphaBoolBlend {blendOption=Linear}
    boneToModify {boneName=dyn_R_backLock_1_01}
    constraintSetup [{targetBone {boneName=head}}, {targetBone {boneName=neck_02}}, {targetBone {boneName=neck_01}}]
    constraintWeights [1,1,1]
   }
   n35 = "Constraint - dyn_R_backLock_1_03 (PPPP)"!(
    ComponentPose=@n34
    Alpha=1
    ConstraintWeights_0=0
    ConstraintWeights_1:float=0.33
    ConstraintWeights_2:float=0.33
    ConstraintWeights_3:float=0.34
   ) -> (Pose) {
    alphaBoolBlend {blendOption=Linear}
    boneToModify {boneName=dyn_R_backLock_1_03}
    constraintSetup [{targetBone {boneName=head}}, {targetBone {boneName=clavicle_r}}, {targetBone {boneName=neck_01}}, {targetBone {boneName=neck_02}}]
    constraintWeights [1,1,1,1]
   }
   n36 = "Constraint - dyn_R_backLock_1_05 (PPPPP)"!(
    ComponentPose=@n35
    Alpha=1
    ConstraintWeights_0=0
    ConstraintWeights_1:float=0.9
    ConstraintWeights_2:float=0
    ConstraintWeights_3:float=0
    ConstraintWeights_4:float=0.1
   ) -> (Pose) {
    alphaBoolBlend {blendOption=Linear}
    boneToModify {boneName=dyn_R_backLock_1_05}
    constraintSetup [
     {targetBone {boneName=spine_05}}
     {targetBone {boneName=clavicle_r}}
     {targetBone {boneName=spine_03}}
     {targetBone {boneName=spine_04}}
     {targetBone {boneName=neck_01}}
    ]
    constraintWeights [1,1,1,1,1]
   }
   n37 = "Constraint - dyn_R_backLock_1_06 (PPPP)"!(
    ComponentPose=@n36
    Alpha=1
    ConstraintWeights_0=0.1
    ConstraintWeights_1:float=0.9
    ConstraintWeights_2:float=0
    ConstraintWeights_3:float=0
   ) -> (Pose) {
    alphaBoolBlend {blendOption=Linear}
    boneToModify {boneName=dyn_R_backLock_1_06}
    constraintSetup [{targetBone {boneName=spine_05}}, {targetBone {boneName=clavicle_r}}, {targetBone {boneName=spine_03}}, {targetBone {boneName=spine_04}}]
    constraintWeights [1,1,1,1]
   }
   n38 = "Constraint - dyn_R_backLock_2_01 (PPPPP)"!(
    ComponentPose=@n37
    Alpha=1
    ConstraintWeights_0=0
    ConstraintWeights_1:float=0.9
    ConstraintWeights_2:float=0
    ConstraintWeights_3:float=0.1
    ConstraintWeights_4:float=0
   ) -> (Pose) {
    alphaBoolBlend {blendOption=Linear}
    boneToModify {boneName=dyn_R_backLock_2_01}
    constraintSetup [
     {targetBone {boneName=spine_05}}
     {targetBone {boneName=head}}
     {targetBone {boneName=neck_01}}
     {targetBone {boneName=neck_02}}
     {targetBone {boneName=clavicle_r}}
    ]
    constraintWeights [1,1,1,1,1]
   }
   n39 = "Constraint - dyn_R_backLock_2_03 (PPPPP)"!(
    ComponentPose=@n38
    Alpha=1
    ConstraintWeights_0=0
    ConstraintWeights_1:float=0.25
    ConstraintWeights_2:float=0.25
    ConstraintWeights_3:float=0.25
    ConstraintWeights_4:float=0.25
   ) -> (Pose) {
    alphaBoolBlend {blendOption=Linear}
    boneToModify {boneName=dyn_R_backLock_2_03}
    constraintSetup [
     {targetBone {boneName=spine_05}}
     {targetBone {boneName=clavicle_r}}
     {targetBone {boneName=neck_01}}
     {targetBone {boneName=neck_02}}
     {targetBone {boneName=head}}
    ]
    constraintWeights [1,1,1,1,1]
   }
   n40 = "Constraint - dyn_R_backLock_2_05 (PP)"!(ComponentPose=@n39, Alpha=1, ConstraintWeights_0=0.1, ConstraintWeights_1:float=0.9) -> (Pose) {
    alphaBoolBlend {blendOption=Linear}
    boneToModify {boneName=dyn_R_backLock_2_05}
    constraintSetup [{targetBone {boneName=spine_05}}, {targetBone {boneName=clavicle_r}}]
    constraintWeights [1,1]
   }
   n41 = "Constraint - dyn_R_backLock_2_06 (PPPPP)"!(
    ComponentPose=@n40
    Alpha=1
    ConstraintWeights_0=0.1
    ConstraintWeights_1:float=0.9
    ConstraintWeights_2:float=0
    ConstraintWeights_3:float=0
    ConstraintWeights_4:float=0
   ) -> (Pose) {
    alphaBoolBlend {blendOption=Linear}
    boneToModify {boneName=dyn_R_backLock_2_06}
    constraintSetup [
     {targetBone {boneName=spine_05}}
     {targetBone {boneName=clavicle_r}}
     {targetBone {boneName=spine_03}}
     {targetBone {boneName=spine_04}}
     {targetBone {boneName=neck_01}}
    ]
    constraintWeights [1,1,1,1,1]
   }
   n42 = RigidBody(ComponentPose=@n41, OverrideWorldGravity=Get GravityOverride()) {
    bClampLinearTranslationLimitToRefPose=true
    bOverrideWorldGravity=true
    overridePhysicsAsset=&/M_MED_EssayViewMyth_Hat_Back_FaceAcc_BackPack_Physics
    simulationSpace=BaseBoneSpace
    alphaBoolBlend {blendOption=Linear}
    baseBoneRef {boneName=spine_05}
    simSpaceSettings {
     dampingAlpha=0
     maxAngularAcceleration=16
     maxAngularVelocity=8
     maxLinearAcceleration=800
     maxLinearVelocity=600
     worldAlpha=1
     externalLinearDragV=<2 0.2 0.2>
    }
   }
   Output Pose(Result=Blend Poses by bool(
     BlendPose_0=Component To Local(ComponentPose=@n42)
     BlendPose_1=Blend Poses by bool(BlendPose_0=@n13, BlendPose_1=@n17, bActiveValue=@n18) {blendType=Linear}
     bActiveValue=Get Backbling_or_FrontEndSkydive() -> (Backbling_or_FrontEndSkydive:bool)
    ) {blendType=Linear}) {layerGroup=None}
   n43 = Get Body_Mesh()
   Save cached pose 'Start Pose'(Pose=Layered blend per bone(
     BasePose=Copy Pose From Mesh() {bUseAttachedParent=true}
     BlendPoses_0=Copy Pose From Mesh(SourceMeshComponent=@n43) {bUseAttachedParent=true}
    ) {layerSetup [{branchFilters [{boneName=upperarm_l}, {boneName=upperarm_r}]}]}) {cacheName=Start Pose}
   n44 = Control Rig!(
    Source=Use cached pose 'Start Pose'()
    Alpha
    Is Crouching:bool=Get bIsCrouching()
    Get is Skydiving Floating:bool=Get bIsSkydiveFloating()
    Is Moving 2D:bool=Get bIsMoving2D()
    Is Crouch Aiming Moving:bool=Get IsCrouchAimingMoving() -> (IsCrouchAimingMoving:bool)
    Backwards Hat:bool=Get BackwardsHat() -> (BackwardsHat:bool)
   ) -> (Pose) {
    alphaBoolBlend {blendOption=Linear}
    controlRigAssetReference {blueprintRigClass=&/M_MED_EssayViewMythHat_FaceAcc_CtrlRig.M_MED_EssayViewMythHat_FaceAcc_CtrlRig_C}
   }
   Save cached pose 'Control Rig Pose'(Pose=@n44) {cacheName=Control Rig Pose}
  ]}
 EventGraph {nodes [
   Event BlueprintUpdateAnimation()
   > n1 = Sequence!(execute) -> (then_0, then_1, then_2:Exec, then_3:Exec, then_4:Exec)
   > Set IsCrouchWalking(IsCrouchWalking:bool=AND Boolean(A=Get bIsCrouching(), B=Get bIsMoving2D())) -> (Output_Get:bool)
   > Set IsCrouchAimingMoving(IsCrouchAimingMoving:bool=AND Boolean!(A=Get bIsCrouching(), B=Get bIsTargeting(), C:bool=Get bIsMoving2D()) -> (ReturnValue)) -> (Output_Get:bool)
   > n2 = Branch(Condition=Get bIsSkyDiving())
   > Set GravityOverride(GravityOverride="0,0,-150")
   > Set!(
    execute
    Sim Space Settings_WorldAlpha:float=0
    Sim Space Settings_VelocityScaleZ:float=1
    Sim Space Settings_DampingAlpha:float=1
    Sim Space Settings_MaxLinearVelocity:float=10000
    Sim Space Settings_MaxAngularVelocity:float=10000
    Sim Space Settings_MaxLinearAcceleration:float=10000
    Sim Space Settings_MaxAngularAcceleration:float=10000
    Sim Space Settings_ExternalLinearDragV:Vector="0,0,0"
    Sim Space Settings_ExternalLinearVelocity:Vector="0,0,0"
    Sim Space Settings_ExternalAngularVelocity:Vector="0,0,0"
   ) -> (then, Output_Get:Sim Space Settings Structure)
   > n3 = Branch(Condition=Get bIsParachuteOpened())
   > Set GravityOverride(GravityOverride="0,0,-980")
   > Set(
    Sim Space Settings_WorldAlpha=0
    Sim Space Settings_VelocityScaleZ=1
    Sim Space Settings_DampingAlpha=1
    Sim Space Settings_MaxLinearVelocity=10000
    Sim Space Settings_MaxAngularVelocity=10000
    Sim Space Settings_MaxLinearAcceleration=10000
    Sim Space Settings_MaxAngularAcceleration=10000
    Sim Space Settings_ExternalLinearDragV="0,0,0"
    Sim Space Settings_ExternalLinearVelocity="0,0,0"
    Sim Space Settings_ExternalAngularVelocity="0,0,0"
   )
   n4 = Branch(execute=@n3.else, Condition=Get bIsSkydiveDiveMode())
   > Set GravityOverride(GravityOverride="0,0,600")
   Set GravityOverride(execute=@n4.else, GravityOverride="0,0,-200")
   > Set(
    Sim Space Settings_WorldAlpha=0
    Sim Space Settings_VelocityScaleZ=1
    Sim Space Settings_DampingAlpha=1
    Sim Space Settings_MaxLinearVelocity=10000
    Sim Space Settings_MaxAngularVelocity=10000
    Sim Space Settings_MaxLinearAcceleration=10000
    Sim Space Settings_MaxAngularAcceleration=10000
    Sim Space Settings_ExternalLinearDragV="0,0,0"
    Sim Space Settings_ExternalLinearVelocity="0,0,0"
    Sim Space Settings_ExternalAngularVelocity="0,0,0"
   )
   Set GravityOverride(execute=@n2.else, GravityOverride="0,0,-980")
   > Set(
    Sim Space Settings_WorldAlpha=1
    Sim Space Settings_VelocityScaleZ=1
    Sim Space Settings_DampingAlpha=1
    Sim Space Settings_MaxLinearVelocity=10000
    Sim Space Settings_MaxAngularVelocity=10000
    Sim Space Settings_MaxLinearAcceleration=10000
    Sim Space Settings_MaxAngularAcceleration=10000
    Sim Space Settings_ExternalLinearDragV="0,0,0"
    Sim Space Settings_ExternalLinearVelocity="0,0,0"
    Sim Space Settings_ExternalAngularVelocity="0,0,0"
   )
   n5 = Cast To FortPlayerPawn(execute=@n1.then_3, Object=GetOwningActor())
   > n6 = Set Character_0(Character_0:Athena Character Item Definition Object Reference=GetItemDefinitionTypeFromLoadoutWithActor!(
     Actor:Actor Object Reference=@n5
     LoadoutSlot:Gameplay Tag Structure=(TagName="LoadoutSlot.Character")
     CosmeticType:Athena Cosmetic Item Definition Class Reference=/Script/FortniteGame.AthenaCharacterItemDefinition
    ) -> (OutCosmeticItemDef:Athena Character Item Definition Object Reference)) -> (Output_Get:Athena Character Item Definition Object Reference)
   > Set isEssayPier(isEssayPier:bool="Equal (Object)"(A=@n6, B=/BRCosmetics/Athena/Items/Cosmetics/Characters/Character_EssayViewPier)) -> (Output_Get:bool)
   > Set isEssayViewMyth(isEssayViewMyth:bool="Equal (Object)"(A=@n6, B=/BRCosmetics/Athena/Items/Cosmetics/Characters/Character_EssayViewMyth)) -> (Output_Get:bool)
   Branch(execute=@n1.then_4, Condition=Get bIsInFrontEnd())
   > Set(
    Sim Space Settings_WorldAlpha=0.25
    Sim Space Settings_VelocityScaleZ=0.25
    Sim Space Settings_DampingAlpha=0.25
    Sim Space Settings_MaxLinearVelocity=10000
    Sim Space Settings_MaxAngularVelocity=10000
    Sim Space Settings_MaxLinearAcceleration=10000
    Sim Space Settings_MaxAngularAcceleration=10000
    Sim Space Settings_ExternalLinearDragV="0,0,0"
    Sim Space Settings_ExternalLinearVelocity="0,0,0"
    Sim Space Settings_ExternalAngularVelocity="0,0,0"
   )
   n7 = CheckForBackbling()
   Set is Wearing Backbling(
    execute=@n1.then_2
    is Wearing Backbling:bool=AND Boolean(A=@n7, B=NOT Boolean(A=@n7.bIsBackblingACape), C=NOT Boolean(A=@n7.bIsBackblingLong))
   ) -> (Output_Get:bool)
   > n8 = Set is Wearing Cape(is Wearing Cape:bool=OR Boolean(A=@n7.bIsBackblingACape, B=@n7.bIsBackblingLong)) -> (Output_Get:bool)
   n9 = Get is Wearing Backbling() -> (is Wearing Backbling:bool)
   n10 = Get bIsInFrontEnd()
   n11 = Get bIsSkyDiving()
   > Set Backbling_or_FrontEndSkydive(Backbling_or_FrontEndSkydive:bool=OR Boolean(A=OR Boolean(A=@n8, B=@n9), B=AND Boolean!(A=@n10, B=@n11) -> (ReturnValue))) -> (Output_Get:bool)
   > Set notFE(notFE:bool=NOT Boolean(A=Get bIsInFrontEnd())) -> (Output_Get:bool)
   n12 = Branch(execute=@n1.then_1, Condition=Get bIsSkyDiving())
   > Set headUp(headUp:double=0) -> (Output_Get:double)
   > Set headUp_Crouching(headUp_Crouching:double=0) -> (Output_Get:double)
   > Set headUp_PostPr(headUp_PostPr:double=0) -> (Output_Get:double)
   > Set headUp_otherStates(headUp_otherStates:double=0) -> (Output_Get:double)
   n13 = Get Skel Mesh()
   n14 = Set headUp(
    execute=@n12.else
    headUp=GetAlphaFromReferenceTransformDelta(SkeletalMeshComponent=@n13, BoneName=spine_05, BaseName=head, InputRange=(X=0,Y=150), TwistAxis="0,0,1")
   )
   > Set headUp_otherStates(headUp_otherStates="Clamp (Float)"(
     Value=GetAlphaFromReferenceTransformDelta(SkeletalMeshComponent=@n13, BoneName=spine_05, BaseName=head, InputRange=(X=0,Y=65), TwistAxis="0,0,1")
     Min=0.5
     Max=0.75
    ))
   > Set headUp_Crouching(headUp_Crouching=float * float(A=@n14, B:bool=Get bIsCrouching()) {is=title K2Node_PromotableOperator})
   > Set headUp_PostPr(headUp_PostPr=float * float(A=Get headUp() -> (headUp:double), B:double=2))
   Event BlueprintInitializeAnimation()
   > Set Skel Mesh(Skel Mesh=GetOwnerSkeletalMeshComponent())
  ]}
]
variables [
 Backbling_or_FrontEndSkydive {type=boolean}
 BackwardsHat {type=boolean}
 Body_Mesh {type=SkeletalMeshComponent}
 Character_0 {type=AthenaCharacterItemDefinition, default=null}
 Crouch Aim Walking {type=boolean}
 GravityOverride {type=Vector, default=<0 0 -980>}
 headUp {type=number}
 headUp_Crouching {type=number}
 headUp_otherStates {type=number}
 headUp_PostPr {type=number}
 is Wearing Backbling {type=boolean}
 is Wearing Cape {type=boolean}
 IsCrouchAimingMoving {type=boolean}
 IsCrouchWalking {type=boolean}
 isEssayPier {type=boolean}
 isEssayViewMyth {type=boolean}
 IsinFrontEndActive {type=boolean}
 notFE {type=boolean}
 Sim Space Settings {
  type=SimSpaceSettings
  default {
   dampingAlpha=0
   maxAngularAcceleration=16
   maxAngularVelocity=8
   maxLinearAcceleration=800
   maxLinearVelocity=600
   velocityScaleZ=1
   worldAlpha=1
   externalLinearDragV=<2 0.2 0.2>
  }
 }
 Skel Mesh {type=SkeletalMeshComponent}
]
